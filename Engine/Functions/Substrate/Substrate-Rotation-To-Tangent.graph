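# Substrate helper material function, exposed to the library, that converts a normalized 0..1 rotation angle into a tangent vector rotated around the surface normal.
# A Custom HLSL node maps the saturated angle to a full turn and blends tangent and bitangent by cosine and sine.
# A MATERIAL_TANGENTSPACENORMAL compile-time branch selects fixed tangent-space axes or a basis built from world-space vertex normal and tangent.
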MaterialFunction /Engine/Functions/Substrate/Substrate-Rotation-To-Tangent
graph {
 properties {
  bExposeToLibrary=true
  description=Convert a rotation angle into a tangent vector
  userExposedCaption=Substrate Rotation-To-Tangent
  libraryCategoriesText [Substrate Helpers]
 }
 nodes [
  n1 = FunctionInput() {
   bUsePreviewValueAsDefault=true
   description="Normalized rotation angle (0..1), defining the rotation aroung the surface normal"
   inputName=Rotation
   inputType=FunctionInput_Scalar
  }
  MaterialExpressionSaturate_0 = Saturate(None=@n1)
  MaterialExpressionVertexNormalWS_0 = VertexNormalWS()
  MaterialExpressionVertexTangentWS_0 = VertexTangentWS()
  n2 = Custom!(
   Rotation=@MaterialExpressionSaturate_0
   SurfaceNormal=@MaterialExpressionVertexNormalWS_0
   SurfaceTangent=@MaterialExpressionVertexTangentWS_0
  ) {
   code="{\r\n\tconst float AngleRad = saturate(Rotation) * PI * 2;\r\n##if MATERIAL_TANGENTSPACENORMAL\r\n\tconst float3 T = float3(1,0,0);\r\n\tconst float3 B = float3(0,-1,0);\r\n\tconst float3 N = float3(0,0,1);\r\n##else\r\n\tconst float3 N = SurfaceNormal;\r\n\tconst float3 T = normalize(SurfaceTangent);\r\n\tconst float3 B = normalize(cross(SurfaceTangent, SurfaceNormal));\r\n##endif\r\n\tconst float3 Out = cos(AngleRad) * T + sin(AngleRad) * B;\r\n\treturn Out;\r\n}"
   inputs [
    {inputName=Rotation, input {expression=&self:MaterialExpressionSaturate_0}}
    {inputName=SurfaceNormal, input {expression=&self:MaterialExpressionVertexNormalWS_0}}
    {inputName=SurfaceTangent, input {expression=&self:MaterialExpressionVertexTangentWS_0}}
   ]
  }
  FunctionOutput(None=@n2) {description=Rotated tangent vector, outputName=Tangent}
 ]
}
