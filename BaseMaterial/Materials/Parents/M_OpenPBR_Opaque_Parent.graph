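# Opaque parent material for Substrate surfaces, feeding texture and parameter layers into the engine's OpenPBR function.
# Static switches set shared coordinates, tangent frame and gradients, with simple UV and dithered triplanar projection available.
# The basecoat combines HSV and tinted color, remapped roughness, metalness, oriented normals, anisotropy, subsurface, thin film and fuzz.
# A topcoat adds separately scaled roughness and normal maps, plus Beer's-law color from thickness.
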Material /BaseMaterial/Materials/Parents/M_OpenPBR_Opaque_Parent
graph {
 expressions {
  MaterialExpressionAdd_2 {
   class=MaterialExpressionAdd
   inputs [
    A {from=MaterialExpressionMaterialFunctionCall_13, output=R}
    B {from=MaterialExpressionScalarParameter_9}
   ]
  }
  MaterialExpressionAppendVector_0 {
   class=MaterialExpressionAppendVector
   inputs [
    A {from=MaterialExpressionNamedRerouteUsage_29, output=""}
    B {from=MaterialExpressionNamedRerouteUsage_29, output=""}
   ]
  }
  MaterialExpressionAppendVector_1 {
   class=MaterialExpressionAppendVector
   inputs [
    A {from=MaterialExpressionNamedRerouteUsage_31, output=""}
    B {from=MaterialExpressionNamedRerouteUsage_31, output=""}
   ]
  }
  MaterialExpressionChannelMaskParameter_0 {
   class=MaterialExpressionChannelMaskParameter
   inputs [None {from=MaterialExpressionReroute_6, output=""}]
   properties {group=04.1 - Basecoat - Subsurface, maskChannel=Alpha, parameterName=Subsurface Channel, sortPriority=20, defaultValue=<rgba 0 0 0 1>}
  }
  MaterialExpressionComponentMask_0 {
   class=MaterialExpressionComponentMask
   inputs [None {from=MaterialExpressionMaterialFunctionCall_4, output=RGB}]
   properties {g=true, r=true}
  }
  MaterialExpressionComponentMask_1 {class=MaterialExpressionComponentMask, inputs [None {from=MaterialExpressionVectorParameter_3}], properties {b=true, g=true, r=true}}
  MaterialExpressionComponentMask_2 {
   class=MaterialExpressionComponentMask
   inputs [None {from=MaterialExpressionMaterialFunctionCall_17, output=RGB}]
   properties {g=true, r=true}
  }
  MaterialExpressionComponentMask_3 {class=MaterialExpressionComponentMask, inputs [None {from=MaterialExpressionMaterialFunctionCall_12, output=RGBA}], properties {a=true}}
  MaterialExpressionComponentMask_4 {class=MaterialExpressionComponentMask, inputs [None {from=MaterialExpressionReroute_6}], properties {a=true}}
  MaterialExpressionComponentMask_5 {class=MaterialExpressionComponentMask, inputs [None {from=MaterialExpressionVectorParameter_9}], properties {g=true, r=true}}
  MaterialExpressionConstant2Vector_0 {class=MaterialExpressionConstant2Vector, properties {g=1, r=-1}}
  MaterialExpressionConstant2Vector_1 {class=MaterialExpressionConstant2Vector, properties {g=-1, r=1}}
  MaterialExpressionConstant2Vector_2 {class=MaterialExpressionConstant2Vector, properties {g=1, r=-1}}
  MaterialExpressionConstant2Vector_3 {class=MaterialExpressionConstant2Vector, properties {g=-1, r=1}}
  MaterialExpressionConstant_0 {class=MaterialExpressionConstant, properties {r=100}}
  MaterialExpressionConstant_1 {class=MaterialExpressionConstant}
  MaterialExpressionConstant_11 {class=MaterialExpressionConstant, properties {r=1}}
  MaterialExpressionConstant_12 {class=MaterialExpressionConstant, properties {r=0.25}}
  MaterialExpressionConstant_2 {class=MaterialExpressionConstant, properties {r=1}}
  MaterialExpressionConstant_7 {class=MaterialExpressionConstant, properties {r=200}}
  MaterialExpressionConstant_8 {class=MaterialExpressionConstant}
  MaterialExpressionConstant_9 {class=MaterialExpressionConstant, properties {r=1}}
  MaterialExpressionConvert_0 {
   class=MaterialExpressionConvert
   inputs [None {from=MaterialExpressionNamedRerouteUsage_13}]
   properties {
    convertInputs [{
      type=Vector4
      defaultValue=<rgba 0 0 0 1>
      expressionInput {
       expression=&self:MaterialExpressionNamedRerouteUsage_13
       inputName=None
       mask=0
       maskA=0
       maskB=0
       maskG=0
       maskR=0
       outputIndex=0
      }
     }]
    convertOutputs [{type=Vector2}, {type=Vector2}]
   }
  }
  MaterialExpressionConvert_1 {
   class=MaterialExpressionConvert
   inputs [None {from=MaterialExpressionNamedRerouteUsage_15}]
   properties {
    convertInputs [{
      type=Vector4
      defaultValue=<rgba 0 0 0 1>
      expressionInput {
       expression=&self:MaterialExpressionNamedRerouteUsage_15
       inputName=None
       mask=0
       maskA=0
       maskB=0
       maskG=0
       maskR=0
       outputIndex=0
      }
     }]
    convertOutputs [{type=Vector2}, {type=Vector2}]
   }
  }
  MaterialExpressionConvert_2 {
   class=MaterialExpressionConvert
   inputs [
    None {from=MaterialExpressionScalarParameter_3}
    None {from=MaterialExpressionScalarParameter_4}
    None {from=MaterialExpressionScalarParameter_5}
    None {from=MaterialExpressionScalarParameter_6}
   ]
   properties {
    convertInputs [
     {
      type=Scalar
      expressionInput {
       expression=&self:MaterialExpressionScalarParameter_3
       inputName=None
       mask=0
       maskA=0
       maskB=0
       maskG=0
       maskR=0
       outputIndex=0
      }
     }
     {
      type=Scalar
      expressionInput {
       expression=&self:MaterialExpressionScalarParameter_4
       inputName=None
       mask=0
       maskA=0
       maskB=0
       maskG=0
       maskR=0
       outputIndex=0
      }
     }
     {
      type=Scalar
      expressionInput {
       expression=&self:MaterialExpressionScalarParameter_5
       inputName=None
       mask=0
       maskA=0
       maskB=0
       maskG=0
       maskR=0
       outputIndex=0
      }
     }
     {
      type=Scalar
      expressionInput {
       expression=&self:MaterialExpressionScalarParameter_6
       inputName=None
       mask=0
       maskA=0
       maskB=0
       maskG=0
       maskR=0
       outputIndex=0
      }
     }
    ]
    convertOutputs [{type=Vector2}, {type=Vector2}]
   }
  }
  MaterialExpressionConvert_23 {
   class=MaterialExpressionConvert
   inputs [None {from=MaterialExpressionNamedRerouteUsage_1}]
   properties {
    convertInputs [{
      type=Vector4
      defaultValue=<rgba 0 0 0 1>
      expressionInput {
       expression=&self:MaterialExpressionNamedRerouteUsage_1
       inputName=None
       mask=0
       maskA=0
       maskB=0
       maskG=0
       maskR=0
       outputIndex=0
      }
     }]
    convertOutputs [{type=Vector2}, {type=Vector2}]
   }
  }
  MaterialExpressionConvert_3 {
   class=MaterialExpressionConvert
   inputs [None {from=MaterialExpressionNamedRerouteUsage_6}]
   properties {
    convertInputs [{
      type=Vector4
      defaultValue=<rgba 0 0 0 1>
      expressionInput {
       expression=&self:MaterialExpressionNamedRerouteUsage_6
       inputName=None
       mask=0
       maskA=0
       maskB=0
       maskG=0
       maskR=0
       outputIndex=0
      }
     }]
    convertOutputs [{type=Vector2}, {type=Vector2}]
   }
  }
  MaterialExpressionConvert_4 {
   class=MaterialExpressionConvert
   inputs [None {from=MaterialExpressionNamedRerouteUsage_8}]
   properties {
    convertInputs [{
      type=Vector4
      defaultValue=<rgba 0 0 0 1>
      expressionInput {
       expression=&self:MaterialExpressionNamedRerouteUsage_8
       inputName=None
       mask=0
       maskA=0
       maskB=0
       maskG=0
       maskR=0
       outputIndex=0
      }
     }]
    convertOutputs [{type=Vector2}, {type=Vector2}]
   }
  }
  MaterialExpressionConvert_5 {
   class=MaterialExpressionConvert
   inputs [None {from=MaterialExpressionNamedRerouteUsage_10}]
   properties {
    convertInputs [{
      type=Vector4
      defaultValue=<rgba 0 0 0 1>
      expressionInput {
       expression=&self:MaterialExpressionNamedRerouteUsage_10
       inputName=None
       mask=0
       maskA=0
       maskB=0
       maskG=0
       maskR=0
       outputIndex=0
      }
     }]
    convertOutputs [{type=Vector2}, {type=Vector2}]
   }
  }
  MaterialExpressionConvert_6 {
   class=MaterialExpressionConvert
   inputs [None {from=MaterialExpressionNamedRerouteUsage_17}]
   properties {
    convertInputs [{
      type=Vector4
      defaultValue=<rgba 0 0 0 1>
      expressionInput {
       expression=&self:MaterialExpressionNamedRerouteUsage_17
       inputName=None
       mask=0
       maskA=0
       maskB=0
       maskG=0
       maskR=0
       outputIndex=0
      }
     }]
    convertOutputs [{type=Vector2}, {type=Vector2}]
   }
  }
  MaterialExpressionConvert_7 {
   class=MaterialExpressionConvert
   inputs [None {from=MaterialExpressionMultiply_14}]
   properties {
    convertInputs [{
      type=Vector4
      defaultValue=<rgba 0 0 0 1>
      expressionInput {
       expression=&self:MaterialExpressionMultiply_14
       inputName=None
       mask=0
       maskA=0
       maskB=0
       maskG=0
       maskR=0
       outputIndex=0
      }
     }]
    convertOutputs [{type=Vector2}, {type=Vector2}]
   }
  }
  MaterialExpressionConvert_8 {
   class=MaterialExpressionConvert
   inputs [None {from=MaterialExpressionNamedRerouteUsage_32}]
   properties {
    convertInputs [{
      type=Vector4
      defaultValue=<rgba 0 0 0 1>
      expressionInput {
       expression=&self:MaterialExpressionNamedRerouteUsage_32
       inputName=None
       mask=0
       maskA=0
       maskB=0
       maskG=0
       maskR=0
       outputIndex=0
      }
     }]
    convertOutputs [{type=Vector2}, {type=Vector2}]
   }
  }
  MaterialExpressionConvert_9 {
   class=MaterialExpressionConvert
   inputs [None {from=MaterialExpressionMultiply_16}]
   properties {
    convertInputs [{
      type=Vector4
      defaultValue=<rgba 0 0 0 1>
      expressionInput {
       expression=&self:MaterialExpressionMultiply_16
       inputName=None
       mask=0
       maskA=0
       maskB=0
       maskG=0
       maskR=0
       outputIndex=0
      }
     }]
    convertOutputs [{type=Vector2}, {type=Vector2}]
   }
  }
  MaterialExpressionCustom_0 {
   class=MaterialExpressionCustom
   inputs [
    WSN {from=MaterialExpressionMaterialFunctionCall_10, output=WSN}
    TSN {from=MaterialExpressionMaterialFunctionCall_10, output=WSN}
   ]
   properties {
    code="{\r\n##if MATERIAL_TANGENTSPACENORMAL\r\n\tconst float3 N = TSN;\r\n##else\r\n\tconst float3 N = WSN;\r\n##endif\r\n\treturn N;\r\n}"
    description=Default Normal
    inputs [
     {
      inputName=WSN
      input {
       expression=&self:MaterialExpressionMaterialFunctionCall_10
       inputName=None
       mask=0
       maskA=0
       maskB=0
       maskG=0
       maskR=0
       outputIndex=0
      }
     }
     {
      inputName=TSN
      input {
       expression=&self:MaterialExpressionMaterialFunctionCall_10
       inputName=None
       mask=0
       maskA=0
       maskB=0
       maskG=0
       maskR=0
       outputIndex=1
      }
     }
    ]
   }
  }
  MaterialExpressionCustom_1 {
   class=MaterialExpressionCustom
   inputs [
    Color {from=MaterialExpressionMaterialFunctionCall_8, output=RGB}
    h {from=MaterialExpressionVectorParameter_6, output=Hue}
    s {from=MaterialExpressionVectorParameter_6, output=Hue}
    v {from=MaterialExpressionVectorParameter_6, output=Hue}
   ]
   properties {
    code="//Color Transform HSV\r\n//https://beesbuzz.biz/code/16-hsv-color-transforms\r\n\r\nfloat vsu = v * s * cos(h * 6.28318530718);\r\nfloat vsw = v * s * sin(h * 6.28318530718);\r\n\r\nfloat3 result = float3( 0.,0.,0. );\r\n\r\nresult.r =  (.299*v + .701*vsu + .168*vsw) * Color.r\r\n        +   (.587*v - .587*vsu + .330*vsw) * Color.g\r\n        +   (.114*v - .114*vsu - .497*vsw) * Color.b;\r\nresult.g =  (.299*v - .299*vsu - .328*vsw) * Color.r\r\n        +   (.587*v + .413*vsu + .035*vsw) * Color.g\r\n        +   (.114*v - .114*vsu + .292*vsw) * Color.b;\r\nresult.b =  (.299*v - .300*vsu + 1.25*vsw) * Color.r\r\n        +   (.587*v - .588*vsu - 1.05*vsw) * Color.g\r\n        +   (.114*v + .886*vsu - .203*vsw) * Color.b;\r\nreturn result;"
    description=TransformHSV
    inputs [
     {
      inputName=Color
      input {
       expression=&self:MaterialExpressionMaterialFunctionCall_8
       inputName=None
       mask=0
       maskA=0
       maskB=0
       maskG=0
       maskR=0
       outputIndex=0
      }
     }
     {
      inputName=h
      input {
       expression=&self:MaterialExpressionVectorParameter_6
       inputName=None
       mask=1
       maskA=0
       maskB=0
       maskG=0
       maskR=1
       outputIndex=1
      }
     }
     {
      inputName=s
      input {
       expression=&self:MaterialExpressionVectorParameter_6
       inputName=None
       mask=1
       maskA=0
       maskB=0
       maskG=1
       maskR=0
       outputIndex=2
      }
     }
     {
      inputName=v
      input {
       expression=&self:MaterialExpressionVectorParameter_6
       inputName=None
       mask=1
       maskA=0
       maskB=1
       maskG=0
       maskR=0
       outputIndex=3
      }
     }
    ]
   }
  }
  MaterialExpressionCustom_2 {
   class=MaterialExpressionCustom
   inputs [
    WSN {from=MaterialExpressionMaterialFunctionCall_18, output=WSN}
    TSN {from=MaterialExpressionMaterialFunctionCall_18, output=WSN}
   ]
   properties {
    code="{\r\n##if MATERIAL_TANGENTSPACENORMAL\r\n\tconst float3 N = TSN;\r\n##else\r\n\tconst float3 N = WSN;\r\n##endif\r\n\treturn N;\r\n}"
    description=Default Normal
    inputs [
     {
      inputName=WSN
      input {
       expression=&self:MaterialExpressionMaterialFunctionCall_18
       inputName=None
       mask=0
       maskA=0
       maskB=0
       maskG=0
       maskR=0
       outputIndex=0
      }
     }
     {
      inputName=TSN
      input {
       expression=&self:MaterialExpressionMaterialFunctionCall_18
       inputName=None
       mask=0
       maskA=0
       maskB=0
       maskG=0
       maskR=0
       outputIndex=1
      }
     }
    ]
   }
  }
  MaterialExpressionDeriveNormalZ_0 {class=MaterialExpressionDeriveNormalZ, inputs [InXY {from=MaterialExpressionMultiply_0, output=""}], properties {}}
  MaterialExpressionDeriveNormalZ_1 {class=MaterialExpressionDeriveNormalZ, inputs [InXY {from=MaterialExpressionMultiply_6, output=""}], properties {}}
  MaterialExpressionDivide_0 {
   class=MaterialExpressionDivide
   inputs [
    A {from=MaterialExpressionMultiply_2}
    B {from=MaterialExpressionConstant_0}
   ]
  }
  MaterialExpressionDivide_1 {
   class=MaterialExpressionDivide
   inputs [
    A {from=MaterialExpressionVectorParameter_1, output=RGB}
    B {from=MaterialExpressionMaterialFunctionCall_23, output=Result}
   ]
  }
  MaterialExpressionIf_4 {
   class=MaterialExpressionIf
   inputs [
    A {from=MaterialExpressionVectorParameter_7, output=Opacity}
    B {from="", output=""}
    "A > B" {from=MaterialExpressionVectorParameter_7, output=Opacity}
    "A == B" {from="", output=""}
    "A < B" {from=MaterialExpressionConstant_11, output=""}
   ]
   properties {constB=0.0033}
  }
  MaterialExpressionLinearInterpolate_0 {
   class=MaterialExpressionLinearInterpolate
   inputs [
    A {from=MaterialExpressionReroute_2}
    B {from=MaterialExpressionSaturate_1}
    Alpha {from=MaterialExpressionVectorParameter_0, output=Tint Strength}
   ]
  }
  MaterialExpressionLinearInterpolate_1 {
   class=MaterialExpressionLinearInterpolate
   inputs [
    A {from=""}
    B {from=MaterialExpressionComponentMask_4}
    Alpha {from=MaterialExpressionScalarParameter_14}
   ]
   properties {constA=1}
  }
  MaterialExpressionLinearInterpolate_2 {
   class=MaterialExpressionLinearInterpolate
   inputs [
    A {from=MaterialExpressionVectorParameter_4, output=RGB}
    B {from=MaterialExpressionVectorParameter_5, output=RGB}
    Alpha {from=MaterialExpressionMaterialFunctionCall_6, output=RGB}
   ]
  }
  MaterialExpressionLinearInterpolate_3 {
   class=MaterialExpressionLinearInterpolate
   inputs [
    A {from=MaterialExpressionMaterialFunctionCall_8, output=RGB}
    B {from=MaterialExpressionSaturate_3}
    Alpha {from=MaterialExpressionReroute_5}
   ]
  }
  MaterialExpressionLinearInterpolate_4 {
   class=MaterialExpressionLinearInterpolate
   inputs [
    A {from=MaterialExpressionScalarParameter_12}
    B {from=MaterialExpressionScalarParameter_19}
    Alpha {from=MaterialExpressionMaterialFunctionCall_9, output=R}
   ]
  }
  MaterialExpressionLinearInterpolate_7 {
   class=MaterialExpressionLinearInterpolate
   inputs [
    A {from=MaterialExpressionScalarParameter_15}
    B {from=MaterialExpressionScalarParameter_26}
    Alpha {from=MaterialExpressionChannelMaskParameter_0, output=R}
   ]
  }
  MaterialExpressionMaterialFunctionCall_0 {
   class=MaterialExpressionMaterialFunctionCall
   inputs [
    ----| Base |---- {from=""}
    base_weight {from=MaterialExpressionMax_0}
    base_color {from=MaterialExpressionLinearInterpolate_0}
    base_diffuse_roughness {from=""}
    base_metalness {from=MaterialExpressionSaturate_4}
    ----| Specular |---- {from=""}
    specular_weight {from=""}
    specular_color {from=MaterialExpressionReroute_8}
    specular_roughness {from=MaterialExpressionSaturate_0}
    specular_ior {from=MaterialExpressionConstant_7}
    specular_roughness_anisotropy {from=MaterialExpressionReroute_14}
    ----| Transmission |---- {from=""}
    transmission_weight {from=""}
    transmission_color {from=""}
    transmission_depth {from=""}
    transmission_scatter {from=""}
    transmission_scatter_anisotropy {from=""}
    transmission_dispersion_scale {from=""}
    transmission_dispersion_abbe_number {from=""}
    ----| Subsurface |---- {from=""}
    subsurface_weight {from=MaterialExpressionMultiply_5}
    subsurface_color {from=MaterialExpressionLinearInterpolate_0}
    subsurface_radius {from=MaterialExpressionLinearInterpolate_7}
    subsurface_radius_scale {from=""}
    subsurface_scatter_anisotropy {from=""}
    ----| Coat |---- {from=""}
    coat_weight {from=MaterialExpressionReroute_10}
    coat_color {from=MaterialExpressionDivide_1}
    coat_roughness {from=MaterialExpressionSaturate_5}
    coat_roughness_anisotropy {from=""}
    coat_ior {from=MaterialExpressionReroute_9}
    coat_darkening {from=""}
    ----| Fuzz |---- {from=""}
    fuzz_weight {from=MaterialExpressionReroute_3}
    fuzz_color {from=MaterialExpressionVectorParameter_10, output=RGB}
    fuzz_roughness {from=MaterialExpressionReroute_1}
    ----| Emission |---- {from=""}
    emission_luminance {from=""}
    emission_color {from=""}
    ----| Thin Film |---- {from=""}
    thin_film_weight {from=""}
    thin_film_thickness {from=MaterialExpressionLinearInterpolate_4}
    thin_film_ior {from=MaterialExpressionScalarParameter_17}
    ----| Geometry |---- {from=""}
    geometry_opacity {from=MaterialExpressionLinearInterpolate_1}
    geometry_thin_walled {from=""}
    geometry_normal {from=MaterialExpressionNamedRerouteDeclaration_6}
    geometry_tangent {from=MaterialExpressionMaterialFunctionCall_11, output=Tangent}
    geometry_coat_normal {from=MaterialExpressionCustom_2}
    geometry_coat_tangent {from=""}
   ]
   properties {materialFunction=/Engine/Functions/Substrate/MF_Substrate_OpenPBR_Opaque}
  }
  MaterialExpressionMaterialFunctionCall_1 {
   class=MaterialExpressionMaterialFunctionCall
   inputs [
    Selector {from=MaterialExpressionScalarParameter_1}
    VectorA {from=MaterialExpressionScalarParameter_0}
    VectorB {from=MaterialExpressionMultiply_1}
    VectorC {from=MaterialExpressionMultiply_3}
    VectorD {from=MaterialExpressionMultiply_4}
   ]
   properties {materialFunction=/BaseMaterial/Materials/Functions/MF_Switch4_Vec2}
  }
  MaterialExpressionMaterialFunctionCall_10 {
   class=MaterialExpressionMaterialFunctionCall
   inputs [
    TSN {from=MaterialExpressionDeriveNormalZ_0}
    Tangent {from=MaterialExpressionNamedRerouteUsage_2}
    Bitangent {from=MaterialExpressionNamedRerouteUsage_3}
    "Rotation (Turns)" {from=MaterialExpressionNamedRerouteUsage_4}
   ]
   properties {materialFunction=/BaseMaterial/Materials/Functions/MF_TangentNormal_OrientWorld}
  }
  MaterialExpressionMaterialFunctionCall_11 {
   class=MaterialExpressionMaterialFunctionCall
   inputs [Rotation {from=MaterialExpressionComponentMask_3}]
   properties {materialFunction=/Engine/Functions/Substrate/Substrate-Rotation-To-Tangent}
  }
  MaterialExpressionMaterialFunctionCall_12 {
   class=MaterialExpressionMaterialFunctionCall
   inputs [
    Coordinates {from=MaterialExpressionNamedRerouteUsage_9}
    TexObject {from=MaterialExpressionTextureObjectParameter_3}
    DDX {from=MaterialExpressionConvert_5}
    DDY {from=MaterialExpressionConvert_5}
    Mask {from=""}
    DefaultUV {from=""}
   ]
   properties {materialFunction=/BaseMaterial/Materials/Functions/MF_TexPointSample_RGB}
  }
  MaterialExpressionMaterialFunctionCall_13 {
   class=MaterialExpressionMaterialFunctionCall
   inputs [
    Coordinates {from=MaterialExpressionNamedRerouteUsage_18}
    TexObject {from=MaterialExpressionTextureObjectParameter_6}
    DDX {from=MaterialExpressionConvert_6}
    DDY {from=MaterialExpressionConvert_6}
    Mask {from=""}
    DefaultUV {from=""}
   ]
   properties {materialFunction=&MF_TexPointSample_RGB}
  }
  MaterialExpressionMaterialFunctionCall_14 {
   class=MaterialExpressionMaterialFunctionCall
   inputs [
    Coordinates {from=MaterialExpressionMultiply_13}
    TexObject {from=MaterialExpressionTextureObjectParameter_7}
    DDX {from=MaterialExpressionConvert_7}
    DDY {from=MaterialExpressionConvert_7}
    Mask {from=""}
    DefaultUV {from=""}
   ]
   properties {materialFunction=&MF_TexPointSample_RGB}
  }
  MaterialExpressionMaterialFunctionCall_15 {
   class=MaterialExpressionMaterialFunctionCall
   inputs [
    Input {from=MaterialExpressionReroute_12}
    Input Low {from=MaterialExpressionConstant_8}
    Input High {from=MaterialExpressionConstant_9}
    Target Low {from=MaterialExpressionScalarParameter_16}
    Target High {from=MaterialExpressionScalarParameter_20}
   ]
   properties {materialFunction=&RemapValueRange}
  }
  MaterialExpressionMaterialFunctionCall_17 {
   class=MaterialExpressionMaterialFunctionCall
   inputs [
    Coordinates {from=MaterialExpressionMultiply_15}
    TexObject {from=MaterialExpressionTextureObjectParameter_9}
    DDX {from=MaterialExpressionConvert_9}
    DDY {from=MaterialExpressionConvert_9}
    Mask {from=""}
    DefaultUV {from=""}
   ]
   properties {materialFunction=/BaseMaterial/Materials/Functions/MF_TexPointSample_NRM}
  }
  MaterialExpressionMaterialFunctionCall_18 {
   class=MaterialExpressionMaterialFunctionCall
   inputs [
    TSN {from=MaterialExpressionDeriveNormalZ_1}
    Tangent {from=MaterialExpressionNamedRerouteUsage_25}
    Bitangent {from=MaterialExpressionNamedRerouteUsage_26}
    "Rotation (Turns)" {from=MaterialExpressionNamedRerouteUsage_27}
   ]
   properties {materialFunction=&MF_TangentNormal_OrientWorld}
  }
  MaterialExpressionMaterialFunctionCall_19 {
   class=MaterialExpressionMaterialFunctionCall
   inputs [
    Selector {from=MaterialExpressionScalarParameter_24}
    VectorA {from=MaterialExpressionScalarParameter_23}
    VectorB {from=MaterialExpressionMultiply_7}
    VectorC {from=MaterialExpressionMultiply_11}
    VectorD {from=MaterialExpressionMultiply_12}
   ]
   properties {materialFunction=&MF_Switch4_Vec2}
  }
  MaterialExpressionMaterialFunctionCall_2 {
   class=MaterialExpressionMaterialFunctionCall
   inputs [
    CoordinateRepeat {from=MaterialExpressionConvert_2}
    CoordinateTransform {from=MaterialExpressionConvert_2}
    "Rotation (Turns)" {from=MaterialExpressionScalarParameter_18}
    UVIndex {from=MaterialExpressionScalarParameter_2}
   ]
   properties {materialFunction=/BaseMaterial/Materials/Functions/MF_ResolveCoordinateFrame_SimpleUV_Basic}
  }
  MaterialExpressionMaterialFunctionCall_23 {
   class=MaterialExpressionMaterialFunctionCall
   inputs [
    Thickness {from=MaterialExpressionVectorParameter_7, output=Thickness}
    Depth Scale {from=MaterialExpressionConstant_12}
   ]
   properties {materialFunction=/Engine/Functions/Engine_MaterialFunctions01/Volumetrics/BeersLaw}
  }
  MaterialExpressionMaterialFunctionCall_24 {
   class=MaterialExpressionMaterialFunctionCall
   inputs [
    Coordinates {from=MaterialExpressionNamedRerouteUsage_14}
    TexObject {from=MaterialExpressionTextureObjectParameter_10}
    DDX {from=MaterialExpressionConvert_8}
    DDY {from=MaterialExpressionConvert_8}
    Mask {from=""}
    DefaultUV {from=""}
   ]
   properties {materialFunction=&MF_TexPointSample_RGB}
  }
  MaterialExpressionMaterialFunctionCall_3 {
   class=MaterialExpressionMaterialFunctionCall
   inputs [
    CoordinateSpace {from=MaterialExpressionScalarParameter_7}
    CoordinateRepeat {from=MaterialExpressionDivide_0}
    CoordinateTransform {from=MaterialExpressionComponentMask_1}
    "Rotation (Turns)" {from=MaterialExpressionScalarParameter_18}
    Dither {from=""}
    BlendFactor {from=MaterialExpressionScalarParameter_8}
   ]
   properties {materialFunction=/BaseMaterial/Materials/Functions/MF_ResolveCoordinateFrame_Triplanar_Dither}
  }
  MaterialExpressionMaterialFunctionCall_4 {
   class=MaterialExpressionMaterialFunctionCall
   inputs [
    Coordinates {from=MaterialExpressionNamedRerouteUsage_0}
    TexObject {from=MaterialExpressionTextureObjectParameter_0}
    DDX {from=MaterialExpressionConvert_23}
    DDY {from=MaterialExpressionConvert_23}
    Mask {from=""}
    DefaultUV {from=""}
   ]
   properties {materialFunction=&MF_TexPointSample_NRM}
  }
  MaterialExpressionMaterialFunctionCall_5 {
   class=MaterialExpressionMaterialFunctionCall
   inputs [
    Coordinates {from=MaterialExpressionNamedRerouteUsage_5}
    TexObject {from=MaterialExpressionTextureObjectParameter_1}
    DDX {from=MaterialExpressionConvert_3}
    DDY {from=MaterialExpressionConvert_3}
    Mask {from=""}
    DefaultUV {from=""}
   ]
   properties {materialFunction=&MF_TexPointSample_RGB}
  }
  MaterialExpressionMaterialFunctionCall_6 {
   class=MaterialExpressionMaterialFunctionCall
   inputs [
    Coordinates {from=MaterialExpressionNamedRerouteUsage_12}
    TexObject {from=MaterialExpressionTextureObjectParameter_4}
    DDX {from=MaterialExpressionConvert_0}
    DDY {from=MaterialExpressionConvert_0}
    Mask {from=""}
    DefaultUV {from=""}
   ]
   properties {materialFunction=&MF_TexPointSample_RGB}
  }
  MaterialExpressionMaterialFunctionCall_7 {
   class=MaterialExpressionMaterialFunctionCall
   inputs [
    Input {from=MaterialExpressionReroute_0}
    Input Low {from=MaterialExpressionConstant_1}
    Input High {from=MaterialExpressionConstant_2}
    Target Low {from=MaterialExpressionScalarParameter_10}
    Target High {from=MaterialExpressionScalarParameter_11}
   ]
   properties {materialFunction=&RemapValueRange}
  }
  MaterialExpressionMaterialFunctionCall_8 {
   class=MaterialExpressionMaterialFunctionCall
   inputs [
    Coordinates {from=MaterialExpressionNamedRerouteUsage_7}
    TexObject {from=MaterialExpressionTextureObjectParameter_2}
    DDX {from=MaterialExpressionConvert_4}
    DDY {from=MaterialExpressionConvert_4}
    Mask {from=""}
    DefaultUV {from=""}
   ]
   properties {materialFunction=&MF_TexPointSample_RGB}
  }
  MaterialExpressionMaterialFunctionCall_9 {
   class=MaterialExpressionMaterialFunctionCall
   inputs [
    Coordinates {from=MaterialExpressionNamedRerouteUsage_16}
    TexObject {from=MaterialExpressionTextureObjectParameter_5}
    DDX {from=MaterialExpressionConvert_1}
    DDY {from=MaterialExpressionConvert_1}
    Mask {from=""}
    DefaultUV {from=""}
   ]
   properties {materialFunction=&MF_TexPointSample_RGB}
  }
  MaterialExpressionMax_0 {
   class=MaterialExpressionMax
   inputs [
    A {from=MaterialExpressionOneMinus_0}
    B {from=MaterialExpressionSaturate_4}
   ]
  }
  MaterialExpressionMultiply_0 {
   class=MaterialExpressionMultiply
   inputs [
    A {from=MaterialExpressionComponentMask_0}
    B {from=MaterialExpressionMaterialFunctionCall_1, output=Selected Vector}
   ]
  }
  MaterialExpressionMultiply_1 {
   class=MaterialExpressionMultiply
   inputs [
    A {from=MaterialExpressionScalarParameter_0}
    B {from=MaterialExpressionConstant2Vector_0}
   ]
  }
  MaterialExpressionMultiply_11 {
   class=MaterialExpressionMultiply
   inputs [
    A {from=MaterialExpressionScalarParameter_23}
    B {from=MaterialExpressionConstant2Vector_3}
   ]
  }
  MaterialExpressionMultiply_12 {
   class=MaterialExpressionMultiply
   inputs [
    A {from=MaterialExpressionScalarParameter_23}
    B {from=""}
   ]
   properties {constB=-1}
  }
  MaterialExpressionMultiply_13 {
   class=MaterialExpressionMultiply
   inputs [
    A {from=MaterialExpressionNamedRerouteUsage_19}
    B {from=MaterialExpressionNamedRerouteUsage_28}
   ]
  }
  MaterialExpressionMultiply_14 {
   class=MaterialExpressionMultiply
   inputs [
    A {from=MaterialExpressionNamedRerouteUsage_20}
    B {from=MaterialExpressionAppendVector_0}
   ]
  }
  MaterialExpressionMultiply_15 {
   class=MaterialExpressionMultiply
   inputs [
    A {from=MaterialExpressionNamedRerouteUsage_23}
    B {from=MaterialExpressionNamedRerouteUsage_30}
   ]
  }
  MaterialExpressionMultiply_16 {
   class=MaterialExpressionMultiply
   inputs [
    A {from=MaterialExpressionNamedRerouteUsage_24}
    B {from=MaterialExpressionAppendVector_1}
   ]
  }
  MaterialExpressionMultiply_17 {
   class=MaterialExpressionMultiply
   inputs [
    A {from=MaterialExpressionComponentMask_5}
    B {from=MaterialExpressionVectorParameter_9, output=Uniform Scale}
   ]
  }
  MaterialExpressionMultiply_2 {
   class=MaterialExpressionMultiply
   inputs [
    A {from=MaterialExpressionVectorParameter_2}
    B {from=MaterialExpressionVectorParameter_2}
   ]
  }
  MaterialExpressionMultiply_3 {
   class=MaterialExpressionMultiply
   inputs [
    A {from=MaterialExpressionScalarParameter_0}
    B {from=MaterialExpressionConstant2Vector_1}
   ]
  }
  MaterialExpressionMultiply_4 {
   class=MaterialExpressionMultiply
   inputs [
    A {from=MaterialExpressionScalarParameter_0}
    B {from=""}
   ]
   properties {constB=-1}
  }
  MaterialExpressionMultiply_5 {
   class=MaterialExpressionMultiply
   inputs [
    A {from=MaterialExpressionScalarParameter_22}
    B {from=MaterialExpressionOneMinus_1}
   ]
  }
  MaterialExpressionMultiply_6 {
   class=MaterialExpressionMultiply
   inputs [
    A {from=MaterialExpressionComponentMask_2}
    B {from=MaterialExpressionMaterialFunctionCall_19, output=Selected Vector}
   ]
  }
  MaterialExpressionMultiply_7 {
   class=MaterialExpressionMultiply
   inputs [
    A {from=MaterialExpressionScalarParameter_23}
    B {from=MaterialExpressionConstant2Vector_2}
   ]
  }
  MaterialExpressionMultiply_8 {
   class=MaterialExpressionMultiply
   inputs [
    A {from=MaterialExpressionLinearInterpolate_3}
    B {from=MaterialExpressionVectorParameter_0, output=RGB}
   ]
  }
  MaterialExpressionMultiply_9 {
   class=MaterialExpressionMultiply
   inputs [
    A {from=MaterialExpressionMaterialFunctionCall_12, output=R}
    B {from=MaterialExpressionScalarParameter_13}
   ]
  }
  MaterialExpressionNamedRerouteDeclaration_0 {
   class=MaterialExpressionNamedRerouteDeclaration
   inputs [None {from=MaterialExpressionStaticSwitch_0}]
   properties {name=Turns, nodeColor=<rgba 0.99 0.0099 0.1079 1>}
  }
  MaterialExpressionNamedRerouteDeclaration_1 {
   class=MaterialExpressionNamedRerouteDeclaration
   inputs [None {from=MaterialExpressionStaticSwitch_1}]
   properties {name=UV, nodeColor=<rgba 0.99 0.0099 0.1079 1>}
  }
  MaterialExpressionNamedRerouteDeclaration_2 {
   class=MaterialExpressionNamedRerouteDeclaration
   inputs [None {from=MaterialExpressionStaticSwitch_2}]
   properties {name=T, nodeColor=<rgba 0.99 0.0099 0.1079 1>}
  }
  MaterialExpressionNamedRerouteDeclaration_3 {
   class=MaterialExpressionNamedRerouteDeclaration
   inputs [None {from=MaterialExpressionStaticSwitch_3}]
   properties {name=B, nodeColor=<rgba 0.99 0.0099 0.1079 1>}
  }
  MaterialExpressionNamedRerouteDeclaration_4 {
   class=MaterialExpressionNamedRerouteDeclaration
   inputs [None {from=MaterialExpressionStaticSwitch_4}]
   properties {name=DXY, nodeColor=<rgba 0.99 0.0099 0.1079 1>}
  }
  MaterialExpressionNamedRerouteDeclaration_6 {
   class=MaterialExpressionNamedRerouteDeclaration
   inputs [None {from=MaterialExpressionCustom_0}]
   properties {name=Normal, nodeColor=<rgba 0 0.3607 1 1>}
  }
  MaterialExpressionNamedRerouteDeclaration_7 {
   class=MaterialExpressionNamedRerouteDeclaration
   inputs [None {from=MaterialExpressionMultiply_17}]
   properties {name=CoatUVMult, nodeColor=<rgba 0.99 0.0099 0.1079 1>}
  }
  MaterialExpressionNamedRerouteUsage_0 {class=MaterialExpressionNamedRerouteUsage}
  MaterialExpressionNamedRerouteUsage_1 {class=MaterialExpressionNamedRerouteUsage}
  MaterialExpressionNamedRerouteUsage_10 {class=MaterialExpressionNamedRerouteUsage}
  MaterialExpressionNamedRerouteUsage_12 {class=MaterialExpressionNamedRerouteUsage}
  MaterialExpressionNamedRerouteUsage_13 {class=MaterialExpressionNamedRerouteUsage}
  MaterialExpressionNamedRerouteUsage_14 {class=MaterialExpressionNamedRerouteUsage}
  MaterialExpressionNamedRerouteUsage_15 {class=MaterialExpressionNamedRerouteUsage}
  MaterialExpressionNamedRerouteUsage_16 {class=MaterialExpressionNamedRerouteUsage}
  MaterialExpressionNamedRerouteUsage_17 {class=MaterialExpressionNamedRerouteUsage}
  MaterialExpressionNamedRerouteUsage_18 {class=MaterialExpressionNamedRerouteUsage}
  MaterialExpressionNamedRerouteUsage_19 {class=MaterialExpressionNamedRerouteUsage}
  MaterialExpressionNamedRerouteUsage_2 {class=MaterialExpressionNamedRerouteUsage}
  MaterialExpressionNamedRerouteUsage_20 {class=MaterialExpressionNamedRerouteUsage}
  MaterialExpressionNamedRerouteUsage_23 {class=MaterialExpressionNamedRerouteUsage}
  MaterialExpressionNamedRerouteUsage_24 {class=MaterialExpressionNamedRerouteUsage}
  MaterialExpressionNamedRerouteUsage_25 {class=MaterialExpressionNamedRerouteUsage}
  MaterialExpressionNamedRerouteUsage_26 {class=MaterialExpressionNamedRerouteUsage}
  MaterialExpressionNamedRerouteUsage_27 {class=MaterialExpressionNamedRerouteUsage}
  MaterialExpressionNamedRerouteUsage_28 {class=MaterialExpressionNamedRerouteUsage}
  MaterialExpressionNamedRerouteUsage_29 {class=MaterialExpressionNamedRerouteUsage}
  MaterialExpressionNamedRerouteUsage_3 {class=MaterialExpressionNamedRerouteUsage}
  MaterialExpressionNamedRerouteUsage_30 {class=MaterialExpressionNamedRerouteUsage}
  MaterialExpressionNamedRerouteUsage_31 {class=MaterialExpressionNamedRerouteUsage}
  MaterialExpressionNamedRerouteUsage_32 {class=MaterialExpressionNamedRerouteUsage}
  MaterialExpressionNamedRerouteUsage_4 {class=MaterialExpressionNamedRerouteUsage}
  MaterialExpressionNamedRerouteUsage_5 {class=MaterialExpressionNamedRerouteUsage}
  MaterialExpressionNamedRerouteUsage_6 {class=MaterialExpressionNamedRerouteUsage}
  MaterialExpressionNamedRerouteUsage_7 {class=MaterialExpressionNamedRerouteUsage}
  MaterialExpressionNamedRerouteUsage_8 {class=MaterialExpressionNamedRerouteUsage}
  MaterialExpressionNamedRerouteUsage_9 {class=MaterialExpressionNamedRerouteUsage}
  MaterialExpressionOneMinus_0 {class=MaterialExpressionOneMinus, inputs [None {from=MaterialExpressionScalarParameter_22}]}
  MaterialExpressionOneMinus_1 {class=MaterialExpressionOneMinus, inputs [None {from=MaterialExpressionSaturate_4}]}
  MaterialExpressionReroute_0 {class=MaterialExpressionReroute, inputs [None {from=MaterialExpressionMaterialFunctionCall_5, output=R}]}
  MaterialExpressionReroute_1 {class=MaterialExpressionReroute, inputs [None {from=MaterialExpressionScalarParameter_21}]}
  MaterialExpressionReroute_10 {class=MaterialExpressionReroute, inputs [None {from=MaterialExpressionSaturate_2}]}
  MaterialExpressionReroute_12 {class=MaterialExpressionReroute, inputs [None {from=MaterialExpressionMaterialFunctionCall_14, output=R}]}
  MaterialExpressionReroute_13 {class=MaterialExpressionReroute, inputs [None {from=MaterialExpressionMultiply_9}]}
  MaterialExpressionReroute_14 {class=MaterialExpressionReroute, inputs [None {from=MaterialExpressionReroute_13}]}
  MaterialExpressionReroute_2 {class=MaterialExpressionReroute, inputs [None {from=MaterialExpressionLinearInterpolate_3}]}
  MaterialExpressionReroute_3 {class=MaterialExpressionReroute, inputs [None {from=MaterialExpressionMaterialFunctionCall_24, output=RGB}]}
  MaterialExpressionReroute_5 {class=MaterialExpressionReroute, inputs [None {from=MaterialExpressionVectorParameter_6, output=HSV Strength}]}
  MaterialExpressionReroute_6 {class=MaterialExpressionReroute, inputs [None {from=MaterialExpressionMaterialFunctionCall_8, output=RGBA}]}
  MaterialExpressionReroute_8 {class=MaterialExpressionReroute, inputs [None {from=MaterialExpressionLinearInterpolate_2}]}
  MaterialExpressionReroute_9 {class=MaterialExpressionReroute, inputs [None {from=MaterialExpressionIf_4}]}
  MaterialExpressionSaturate_0 {class=MaterialExpressionSaturate, inputs [None {from=MaterialExpressionMaterialFunctionCall_7, output=Result}]}
  MaterialExpressionSaturate_1 {class=MaterialExpressionSaturate, inputs [None {from=MaterialExpressionMultiply_8}]}
  MaterialExpressionSaturate_2 {class=MaterialExpressionSaturate, inputs [None {from=MaterialExpressionVectorParameter_7, output=Opacity}]}
  MaterialExpressionSaturate_3 {class=MaterialExpressionSaturate, inputs [None {from=MaterialExpressionCustom_1}]}
  MaterialExpressionSaturate_4 {class=MaterialExpressionSaturate, inputs [None {from=MaterialExpressionAdd_2}]}
  MaterialExpressionSaturate_5 {class=MaterialExpressionSaturate, inputs [None {from=MaterialExpressionMaterialFunctionCall_15, output=Result}]}
  MaterialExpressionScalarParameter_0 {
   class=MaterialExpressionScalarParameter
   properties {defaultValue=1, group=04.4 - Basecoat - Normal, parameterName=Normal Map Strength, sliderMax=2, sliderMin=-2, sortPriority=10}
  }
  MaterialExpressionScalarParameter_1 {
   class=MaterialExpressionScalarParameter
   properties {
    controlType=Enumeration
    enumeration=/BaseMaterial/Materials/Utilities/ENormalFlip
    group=04.4 - Basecoat - Normal
    parameterName=Normal Map Channel Flip
    sortPriority=20
   }
  }
  MaterialExpressionScalarParameter_10 {class=MaterialExpressionScalarParameter, properties {group=04.3 - Basecoat - Roughness, parameterName=Roughness Min, sortPriority=10}}
  MaterialExpressionScalarParameter_11 {
   class=MaterialExpressionScalarParameter
   properties {defaultValue=1, group=04.3 - Basecoat - Roughness, parameterName=Roughness Max, sortPriority=11}
  }
  MaterialExpressionScalarParameter_12 {
   class=MaterialExpressionScalarParameter
   properties {group=04.5 - Basecoat - Thin Film, parameterName=Thin Film Thickness Min, sortPriority=10}
  }
  MaterialExpressionScalarParameter_13 {class=MaterialExpressionScalarParameter, properties {group=04.6 - Basecoat - Anisotropy, parameterName=Anisotropy, sortPriority=20}}
  MaterialExpressionScalarParameter_14 {class=MaterialExpressionScalarParameter, properties {group=04.1 - Basecoat - Color, parameterName=Cutout Opacity, sortPriority=30}}
  MaterialExpressionScalarParameter_15 {
   class=MaterialExpressionScalarParameter
   properties {defaultValue=1, group=04.1 - Basecoat - Subsurface, parameterName=Subsurface Radius Min, sliderMax=10, sortPriority=10}
  }
  MaterialExpressionScalarParameter_16 {
   class=MaterialExpressionScalarParameter
   properties {defaultValue=0.15, group=01.1 - Topcoat, parameterName=Topcoat Roughness Min, sliderMax=1, sortPriority=30}
  }
  MaterialExpressionScalarParameter_17 {
   class=MaterialExpressionScalarParameter
   properties {defaultValue=1.55, group=04.5 - Basecoat - Thin Film, parameterName=Thin Film IOR, sortPriority=20}
  }
  MaterialExpressionScalarParameter_18 {class=MaterialExpressionScalarParameter, properties {group=99.1 - UV Transforms, parameterName=Rotation, sortPriority=25}}
  MaterialExpressionScalarParameter_19 {
   class=MaterialExpressionScalarParameter
   properties {group=04.5 - Basecoat - Thin Film, parameterName=Thin Film Thickness Max, sortPriority=11}
  }
  MaterialExpressionScalarParameter_2 {
   class=MaterialExpressionScalarParameter
   properties {
    controlType=Enumeration
    enumeration=/BaseMaterial/Materials/Utilities/EUVCoordinateIndex
    group=99.1 - UV Transforms
    parameterName=UV Index
    sortPriority=0
   }
  }
  MaterialExpressionScalarParameter_20 {
   class=MaterialExpressionScalarParameter
   properties {defaultValue=0.15, group=01.1 - Topcoat, parameterName=Topcoat Roughness Max, sliderMax=1, sortPriority=31}
  }
  MaterialExpressionScalarParameter_21 {class=MaterialExpressionScalarParameter, properties {group=05.1 - Fuzz, parameterName=Fuzz Roughness, sliderMax=1, sortPriority=20}}
  MaterialExpressionScalarParameter_22 {
   class=MaterialExpressionScalarParameter
   properties {group=04.1 - Basecoat - Subsurface, parameterName=Subsurface Strength, sliderMax=1, sortPriority=0}
  }
  MaterialExpressionScalarParameter_23 {
   class=MaterialExpressionScalarParameter
   properties {defaultValue=1, group=01.2 - Topcoat - Normal, parameterName=Topcoat Normal Strength, sliderMax=2, sliderMin=-2, sortPriority=10}
  }
  MaterialExpressionScalarParameter_24 {
   class=MaterialExpressionScalarParameter
   properties {
    controlType=Enumeration
    enumeration=&ENormalFlip
    group=01.2 - Topcoat - Normal
    parameterName=Topcoat Normal Channel Flip
    sortPriority=20
   }
  }
  MaterialExpressionScalarParameter_26 {
   class=MaterialExpressionScalarParameter
   properties {defaultValue=1, group=04.1 - Basecoat - Subsurface, parameterName=Subsurface Radius Max, sliderMax=10, sortPriority=11}
  }
  MaterialExpressionScalarParameter_3 {
   class=MaterialExpressionScalarParameter
   properties {defaultValue=1, group=99.1 - UV Transforms, parameterName=Tile U, sliderMax=10, sortPriority=10}
  }
  MaterialExpressionScalarParameter_4 {
   class=MaterialExpressionScalarParameter
   properties {defaultValue=1, group=99.1 - UV Transforms, parameterName=Tile V, sliderMax=10, sortPriority=11}
  }
  MaterialExpressionScalarParameter_5 {
   class=MaterialExpressionScalarParameter
   properties {group=99.1 - UV Transforms, parameterName=Offset U, sliderMax=1, sliderMin=-1, sortPriority=20}
  }
  MaterialExpressionScalarParameter_6 {
   class=MaterialExpressionScalarParameter
   properties {group=99.1 - UV Transforms, parameterName=Offset V, sliderMax=1, sliderMin=-1, sortPriority=21}
  }
  MaterialExpressionScalarParameter_7 {
   class=MaterialExpressionScalarParameter
   properties {
    controlType=Enumeration
    defaultValue=1
    enumeration=/BaseMaterial/Materials/Utilities/ETriplanarCoordinateFrame
    group=99.1 - UV Transforms
    parameterName=Coordinate Space
    sortPriority=0
   }
  }
  MaterialExpressionScalarParameter_8 {
   class=MaterialExpressionScalarParameter
   properties {defaultValue=0.5, group=99.1 - UV Transforms, parameterName=Transition Softness, sliderMax=1, sortPriority=30}
  }
  MaterialExpressionScalarParameter_9 {
   class=MaterialExpressionScalarParameter
   properties {group=04.1 - Basecoat - Metalness, parameterName=Metalness, sliderMax=1, sliderMin=-1, sortPriority=10}
  }
  MaterialExpressionStaticBoolParameter_0 {class=MaterialExpressionStaticBoolParameter, inputs [], properties {group=99.1 - UV Transforms, parameterName=Use UVs, sortPriority=-10}}
  MaterialExpressionStaticSwitch_0 {
   class=MaterialExpressionStaticSwitch
   inputs [
    True {from=MaterialExpressionMaterialFunctionCall_2, output="Rotation (Turns)"}
    False {from=MaterialExpressionMaterialFunctionCall_3, output="Rotation (Turns)"}
    Value {from=MaterialExpressionStaticBoolParameter_0, output=""}
   ]
   properties {}
  }
  MaterialExpressionStaticSwitch_1 {
   class=MaterialExpressionStaticSwitch
   inputs [
    True {from=MaterialExpressionMaterialFunctionCall_2, output=UV}
    False {from=MaterialExpressionMaterialFunctionCall_3, output=UV}
    Value {from=MaterialExpressionStaticBoolParameter_0, output=""}
   ]
   properties {}
  }
  MaterialExpressionStaticSwitch_2 {
   class=MaterialExpressionStaticSwitch
   inputs [
    True {from=MaterialExpressionMaterialFunctionCall_2, output=Tangent}
    False {from=MaterialExpressionMaterialFunctionCall_3, output=Tangent}
    Value {from=MaterialExpressionStaticBoolParameter_0, output=""}
   ]
   properties {}
  }
  MaterialExpressionStaticSwitch_3 {
   class=MaterialExpressionStaticSwitch
   inputs [
    True {from=MaterialExpressionMaterialFunctionCall_2, output=Bitangent}
    False {from=MaterialExpressionMaterialFunctionCall_3, output=Bitangent}
    Value {from=MaterialExpressionStaticBoolParameter_0, output=""}
   ]
   properties {}
  }
  MaterialExpressionStaticSwitch_4 {
   class=MaterialExpressionStaticSwitch
   inputs [
    True {from=MaterialExpressionMaterialFunctionCall_2, output=DXY}
    False {from=MaterialExpressionMaterialFunctionCall_3, output=DXY}
    Value {from=MaterialExpressionStaticBoolParameter_0, output=""}
   ]
   properties {}
  }
  MaterialExpressionTextureObjectParameter_0 {
   class=MaterialExpressionTextureObjectParameter
   inputs [
    UVs {from="", output=""}
    Apply View MipBias {from="", output=""}
   ]
   properties {
    group=04.4 - Basecoat - Normal
    parameterName=Normal Map
    samplerType=SAMPLERTYPE_Normal
    sortPriority=0
    texture=/BaseMaterial/Textures/Default/T_FlatNorm_BC5
   }
  }
  MaterialExpressionTextureObjectParameter_1 {
   class=MaterialExpressionTextureObjectParameter
   inputs [
    UVs {from="", output=""}
    Apply View MipBias {from="", output=""}
   ]
   properties {
    group=04.3 - Basecoat - Roughness
    parameterName=Roughness Map
    samplerType=SAMPLERTYPE_LinearColor
    sortPriority=0
    texture=/BaseMaterial/Textures/Default/T_BlackLinearColor
   }
  }
  MaterialExpressionTextureObjectParameter_10 {
   class=MaterialExpressionTextureObjectParameter
   inputs [
    UVs {from="", output=""}
    Apply View MipBias {from="", output=""}
   ]
   properties {
    group=05.1 - Fuzz
    parameterName=Fuzz Map
    samplerType=SAMPLERTYPE_LinearColor
    sortPriority=0
    texture=&T_BlackLinearColor
   }
  }
  MaterialExpressionTextureObjectParameter_2 {
   class=MaterialExpressionTextureObjectParameter
   inputs [
    UVs {from="", output=""}
    Apply View MipBias {from="", output=""}
   ]
   properties {group=04.1 - Basecoat - Color, parameterName=Color Map, sortPriority=0, texture=/BaseMaterial/Textures/Default/T_Gray_sRGB}
  }
  MaterialExpressionTextureObjectParameter_3 {
   class=MaterialExpressionTextureObjectParameter
   inputs [
    UVs {from="", output=""}
    Apply View MipBias {from="", output=""}
   ]
   properties {
    group=04.6 - Basecoat - Anisotropy
    parameterName=Anisotropy Map
    samplerType=SAMPLERTYPE_LinearColor
    sortPriority=0
    texture=&T_BlackLinearColor
   }
  }
  MaterialExpressionTextureObjectParameter_4 {
   class=MaterialExpressionTextureObjectParameter
   inputs [
    UVs {from="", output=""}
    Apply View MipBias {from="", output=""}
   ]
   properties {
    group=04.2 - Basecoat - Specular
    parameterName=Spec Map
    samplerType=SAMPLERTYPE_LinearColor
    sortPriority=0
    texture=&T_BlackLinearColor
   }
  }
  MaterialExpressionTextureObjectParameter_5 {
   class=MaterialExpressionTextureObjectParameter
   inputs [
    UVs {from="", output=""}
    Apply View MipBias {from="", output=""}
   ]
   properties {
    group=04.5 - Basecoat - Thin Film
    parameterName=Thickness Map
    samplerType=SAMPLERTYPE_LinearColor
    sortPriority=0
    texture=&T_BlackLinearColor
   }
  }
  MaterialExpressionTextureObjectParameter_6 {
   class=MaterialExpressionTextureObjectParameter
   inputs [
    UVs {from="", output=""}
    Apply View MipBias {from="", output=""}
   ]
   properties {
    group=04.1 - Basecoat - Metalness
    parameterName=Metalness Map
    samplerType=SAMPLERTYPE_LinearColor
    sortPriority=0
    texture=&T_BlackLinearColor
   }
  }
  MaterialExpressionTextureObjectParameter_7 {
   class=MaterialExpressionTextureObjectParameter
   inputs [
    UVs {from="", output=""}
    Apply View MipBias {from="", output=""}
   ]
   properties {
    group=01.1 - Topcoat
    parameterName=Topcoat Roughness Map
    samplerType=SAMPLERTYPE_LinearColor
    sortPriority=20
    texture=&T_BlackLinearColor
   }
  }
  MaterialExpressionTextureObjectParameter_9 {
   class=MaterialExpressionTextureObjectParameter
   inputs [
    UVs {from="", output=""}
    Apply View MipBias {from="", output=""}
   ]
   properties {
    group=01.2 - Topcoat - Normal
    parameterName=Topcoat Normal Map
    samplerType=SAMPLERTYPE_Normal
    sortPriority=0
    texture=&T_FlatNorm_BC5
   }
  }
  MaterialExpressionVectorParameter_0 {
   class=MaterialExpressionVectorParameter
   properties {
    group=04.1 - Basecoat - Color
    parameterName=Tint
    sortPriority=20
    channelNames {a=Tint Strength, b="", g="", r=""}
    defaultValue=<rgba 1 1 1 0>
   }
  }
  MaterialExpressionVectorParameter_1 {
   class=MaterialExpressionVectorParameter
   properties {
    group=01.1 - Topcoat
    parameterName=Topcoat Color
    sortPriority=0
    channelNames {a=None, b="", g="", r=""}
    defaultValue=<rgba 0.975 0.975 0.975 1>
   }
  }
  MaterialExpressionVectorParameter_10 {
   class=MaterialExpressionVectorParameter
   properties {group=05.1 - Fuzz, parameterName=Fuzz Color, sortPriority=10, channelNames {a=None, b="", g="", r=""}, defaultValue=<rgba 0.8 0.8 0.8 1>}
  }
  MaterialExpressionVectorParameter_2 {
   class=MaterialExpressionVectorParameter
   properties {
    group=99.1 - UV Transforms
    parameterName=Tile XYZ
    sortPriority=10
    channelNames {a=Uniform Scale, b=Tile Z, g=Tile Y, r=Tile X}
    defaultValue=<rgba 1 1 1 0.7812>
   }
  }
  MaterialExpressionVectorParameter_3 {
   class=MaterialExpressionVectorParameter
   properties {
    group=99.1 - UV Transforms
    parameterName=Offset XYZ
    sortPriority=20
    channelNames {a=None, b=Translate Z, g=Translate Y, r=Translate X}
    defaultValue=<rgba 0 0 0 1>
   }
  }
  MaterialExpressionVectorParameter_4 {
   class=MaterialExpressionVectorParameter
   properties {group=04.2 - Basecoat - Specular, parameterName=F0 Min, sortPriority=10, defaultValue=<rgba 0.04 0.04 0.04 1>}
  }
  MaterialExpressionVectorParameter_5 {
   class=MaterialExpressionVectorParameter
   properties {group=04.2 - Basecoat - Specular, parameterName=F0 Max, sortPriority=11, defaultValue=<rgba 0.06 0.06 0.06 1>}
  }
  MaterialExpressionVectorParameter_6 {
   class=MaterialExpressionVectorParameter
   properties {group=04.1 - Basecoat - Color, parameterName=Color Map HSV, sortPriority=10, channelNames {a=HSV Strength, b=Value, g=Saturation, r=Hue}}
  }
  MaterialExpressionVectorParameter_7 {
   class=MaterialExpressionVectorParameter
   properties {
    group=01.1 - Topcoat
    parameterName=Topcoat Properties
    sortPriority=10
    channelNames {a=None, b=IOR, g=Opacity, r=Thickness}
    defaultValue=<rgba 0.5 1 1.55 1>
   }
  }
  MaterialExpressionVectorParameter_9 {
   class=MaterialExpressionVectorParameter
   properties {
    group=99.1 - UV Transforms
    parameterName=Topcoat Tile
    sortPriority=40
    channelNames {a=None, b=Uniform Scale, g=Tile Y, r=Tile X}
    defaultValue=<rgba 1 1 1 1>
   }
  }
 }
 properties {
  bUsedWithGeometryCollections=true
  bUsedWithInstancedStaticMeshes=true
  bUsedWithMeshDeformer=true
  bUsedWithNanite=true
  bUsedWithSkeletalMesh=true
  shadingModel=MSM_SubsurfaceProfile
  usedShadingModels=Subsurface Profile
 }
}
