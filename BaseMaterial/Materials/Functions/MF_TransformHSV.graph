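# Material function that shifts a color's hue, saturation and value, outputting Out_Color clamped to 0–1.
# Hue sine and cosine, scaled by saturation and value, form VSU and VSW terms, avoiding an RGB-to-HSV round trip.
# Three InverseTransformMatrix calls with constant luma-based coefficients build per-channel weights, which are dotted with the input color.
# An unconnected Custom HLSL node keeps the reference formula and its beesbuzz.biz source.
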MaterialFunction /BaseMaterial/Materials/Functions/MF_TransformHSV
graph {expressions {
  MaterialExpressionConstant3Vector_10 {class=MaterialExpressionConstant3Vector, properties {constant=<rgba 0.299 -0.299 1.25 0>}}
  MaterialExpressionConstant3Vector_11 {class=MaterialExpressionConstant3Vector, properties {constant=<rgba 0.587 -0.587 -1.05 0>}}
  MaterialExpressionConstant3Vector_12 {class=MaterialExpressionConstant3Vector, properties {constant=<rgba 0.114 0.886 -0.203 0>}}
  MaterialExpressionConstant3Vector_3 {class=MaterialExpressionConstant3Vector, properties {constant=<rgba 0.299 0.701 0.168 0>}}
  MaterialExpressionConstant3Vector_5 {class=MaterialExpressionConstant3Vector, properties {constant=<rgba 0.587 -0.587 0.33 0>}}
  MaterialExpressionConstant3Vector_6 {class=MaterialExpressionConstant3Vector, properties {constant=<rgba 0.114 -0.114 -0.497 0>}}
  MaterialExpressionConstant3Vector_7 {class=MaterialExpressionConstant3Vector, properties {constant=<rgba 0.299 -0.299 -0.328 0>}}
  MaterialExpressionConstant3Vector_8 {class=MaterialExpressionConstant3Vector, properties {constant=<rgba 0.587 0.413 0.035 0>}}
  MaterialExpressionConstant3Vector_9 {class=MaterialExpressionConstant3Vector, properties {constant=<rgba 0.114 -0.114 0.292 0>}}
  MaterialExpressionConvert_5 {
   class=MaterialExpressionConvert
   inputs [
    None {from=MaterialExpressionNamedRerouteUsage_6}
    None {from=MaterialExpressionNamedRerouteUsage_7}
    None {from=MaterialExpressionNamedRerouteUsage_8}
   ]
   properties {
    nodeName=Make Vector3
    convertInputs [
     {
      type=Scalar
      expressionInput {
       expression=&self:MaterialExpressionNamedRerouteUsage_6
       inputName=None
       mask=0
       maskA=0
       maskB=0
       maskG=0
       maskR=0
       outputIndex=0
      }
     }
     {
      type=Scalar
      expressionInput {
       expression=&self:MaterialExpressionNamedRerouteUsage_7
       inputName=None
       mask=0
       maskA=0
       maskB=0
       maskG=0
       maskR=0
       outputIndex=0
      }
     }
     {
      type=Scalar
      expressionInput {
       expression=&self:MaterialExpressionNamedRerouteUsage_8
       inputName=None
       mask=0
       maskA=0
       maskB=0
       maskG=0
       maskR=0
       outputIndex=0
      }
     }
    ]
    convertOutputs [{type=Vector3}]
   }
  }
  MaterialExpressionConvert_6 {
   class=MaterialExpressionConvert
   inputs [
    None {from=MaterialExpressionDotProduct_1}
    None {from=MaterialExpressionDotProduct_2}
    None {from=MaterialExpressionDotProduct_3}
   ]
   properties {
    nodeName=Make Vector3
    convertInputs [
     {
      type=Scalar
      expressionInput {
       expression=&self:MaterialExpressionDotProduct_1
       inputName=None
       mask=0
       maskA=0
       maskB=0
       maskG=0
       maskR=0
       outputIndex=0
      }
     }
     {
      type=Scalar
      expressionInput {
       expression=&self:MaterialExpressionDotProduct_2
       inputName=None
       mask=0
       maskA=0
       maskB=0
       maskG=0
       maskR=0
       outputIndex=0
      }
     }
     {
      type=Scalar
      expressionInput {
       expression=&self:MaterialExpressionDotProduct_3
       inputName=None
       mask=0
       maskA=0
       maskB=0
       maskG=0
       maskR=0
       outputIndex=0
      }
     }
    ]
    convertOutputs [{type=Vector3}]
   }
  }
  MaterialExpressionCosine_0 {class=MaterialExpressionCosine, inputs [None {from=MaterialExpressionFunctionInput_1, output=""}], properties {}}
  MaterialExpressionCustom_1 {
   class=MaterialExpressionCustom
   inputs [
    Color {from="", output=""}
    h {from="", output=""}
    s {from="", output=""}
    v {from="", output=""}
   ]
   properties {
    code="//Color Transform HSV\r\n//https://beesbuzz.biz/code/16-hsv-color-transforms\r\n//HLSL as ref for node functionality.\r\n\r\nfloat vsu = v * s * cos(h * 6.28318530718);\r\nfloat vsw = v * s * sin(h * 6.28318530718);\r\n\r\nfloat3 result = float3( 0.,0.,0. );\r\n\r\nresult.r =  (.299*v + .701*vsu + .168*vsw) * Color.r\r\n        +   (.587*v - .587*vsu + .330*vsw) * Color.g\r\n        +   (.114*v - .114*vsu - .497*vsw) * Color.b;\r\nresult.g =  (.299*v - .299*vsu - .328*vsw) * Color.r\r\n        +   (.587*v + .413*vsu + .035*vsw) * Color.g\r\n        +   (.114*v - .114*vsu + .292*vsw) * Color.b;\r\nresult.b =  (.299*v - .300*vsu + 1.25*vsw) * Color.r\r\n        +   (.587*v - .588*vsu - 1.05*vsw) * Color.g\r\n        +   (.114*v + .886*vsu - .203*vsw) * Color.b;\r\nreturn result;"
    description=TransformHSV
    showCode=true
    inputs [
     {inputName=Color, input {expression=null, inputName=None, mask=0, maskA=0, maskB=0, maskG=0, maskR=0, outputIndex=0}}
     {inputName=h, input {expression=null, inputName=None, mask=1, maskA=0, maskB=0, maskG=0, maskR=1, outputIndex=1}}
     {inputName=s, input {expression=null, inputName=None, mask=1, maskA=0, maskB=0, maskG=1, maskR=0, outputIndex=2}}
     {inputName=v, input {expression=null, inputName=None, mask=1, maskA=0, maskB=1, maskG=0, maskR=0, outputIndex=3}}
    ]
   }
  }
  MaterialExpressionDotProduct_1 {
   class=MaterialExpressionDotProduct
   inputs [
    A {from=MaterialExpressionNamedRerouteUsage_18, output=""}
    B {from=MaterialExpressionMaterialFunctionCall_2, output=Transformed Vector 3}
   ]
   properties {}
  }
  MaterialExpressionDotProduct_2 {
   class=MaterialExpressionDotProduct
   inputs [
    A {from=MaterialExpressionNamedRerouteUsage_19, output=""}
    B {from=MaterialExpressionMaterialFunctionCall_4, output=Transformed Vector 3}
   ]
   properties {}
  }
  MaterialExpressionDotProduct_3 {
   class=MaterialExpressionDotProduct
   inputs [
    A {from=MaterialExpressionNamedRerouteUsage_20, output=""}
    B {from=MaterialExpressionMaterialFunctionCall_5, output=Transformed Vector 3}
   ]
   properties {}
  }
  MaterialExpressionFunctionInput_0 {
   class=MaterialExpressionFunctionInput
   inputs [Preview {from="", output=""}]
   properties {bUsePreviewValueAsDefault=true, inputName=In_Color, previewValue=<0.18 0.18 0.18 1>}
  }
  MaterialExpressionFunctionInput_1 {
   class=MaterialExpressionFunctionInput
   inputs [Preview {from="", output=""}]
   properties {bUsePreviewValueAsDefault=true, inputName=In_Hue, inputType=FunctionInput_Scalar, sortPriority=10}
  }
  MaterialExpressionFunctionInput_2 {
   class=MaterialExpressionFunctionInput
   inputs [Preview {from="", output=""}]
   properties {bUsePreviewValueAsDefault=true, inputName=In_Saturation, inputType=FunctionInput_Scalar, sortPriority=20, previewValue=<1 0 0 1>}
  }
  MaterialExpressionFunctionInput_3 {
   class=MaterialExpressionFunctionInput
   inputs [Preview {from="", output=""}]
   properties {bUsePreviewValueAsDefault=true, inputName=In_Value, inputType=FunctionInput_Scalar, sortPriority=30, previewValue=<1 0 0 1>}
  }
  MaterialExpressionFunctionOutput_0 {class=MaterialExpressionFunctionOutput, inputs [None {from=MaterialExpressionSaturate_0, output=""}], properties {outputName=Out_Color}}
  MaterialExpressionMaterialFunctionCall_2 {
   class=MaterialExpressionMaterialFunctionCall
   inputs [
    VectorToTransform {from=MaterialExpressionConvert_5}
    BasisX {from=MaterialExpressionConstant3Vector_3}
    BasisY {from=MaterialExpressionConstant3Vector_5}
    BasisZ {from=MaterialExpressionConstant3Vector_6}
   ]
   properties {materialFunction=/Engine/Functions/Engine_MaterialFunctions02/Math/InverseTransformMatrix}
  }
  MaterialExpressionMaterialFunctionCall_4 {
   class=MaterialExpressionMaterialFunctionCall
   inputs [
    VectorToTransform {from=MaterialExpressionConvert_5}
    BasisX {from=MaterialExpressionConstant3Vector_7}
    BasisY {from=MaterialExpressionConstant3Vector_8}
    BasisZ {from=MaterialExpressionConstant3Vector_9}
   ]
   properties {materialFunction=&InverseTransformMatrix}
  }
  MaterialExpressionMaterialFunctionCall_5 {
   class=MaterialExpressionMaterialFunctionCall
   inputs [
    VectorToTransform {from=MaterialExpressionConvert_5}
    BasisX {from=MaterialExpressionConstant3Vector_10}
    BasisY {from=MaterialExpressionConstant3Vector_11}
    BasisZ {from=MaterialExpressionConstant3Vector_12}
   ]
   properties {materialFunction=&InverseTransformMatrix}
  }
  MaterialExpressionMultiply_0 {
   class=MaterialExpressionMultiply
   inputs [
    A {from=MaterialExpressionCosine_0}
    B {from=MaterialExpressionNamedRerouteUsage_2}
   ]
  }
  MaterialExpressionMultiply_1 {
   class=MaterialExpressionMultiply
   inputs [
    A {from=MaterialExpressionMultiply_0}
    B {from=MaterialExpressionNamedRerouteUsage_3}
   ]
  }
  MaterialExpressionMultiply_2 {
   class=MaterialExpressionMultiply
   inputs [
    A {from=MaterialExpressionSine_0}
    B {from=MaterialExpressionNamedRerouteUsage_4}
   ]
  }
  MaterialExpressionMultiply_3 {
   class=MaterialExpressionMultiply
   inputs [
    A {from=MaterialExpressionMultiply_2}
    B {from=MaterialExpressionNamedRerouteUsage_5}
   ]
  }
  MaterialExpressionNamedRerouteDeclaration_1 {
   class=MaterialExpressionNamedRerouteDeclaration
   inputs [None {from=MaterialExpressionFunctionInput_2}]
   properties {name=S, nodeColor=<rgba 0 0 0 1>}
  }
  MaterialExpressionNamedRerouteDeclaration_17 {
   class=MaterialExpressionNamedRerouteDeclaration
   inputs [None {from=MaterialExpressionFunctionInput_0}]
   properties {name=Color, nodeColor=<rgba 1 0.6094 0 1>}
  }
  MaterialExpressionNamedRerouteDeclaration_2 {
   class=MaterialExpressionNamedRerouteDeclaration
   inputs [None {from=MaterialExpressionFunctionInput_3}]
   properties {name=V, nodeColor=<rgba 1 1 1 1>}
  }
  MaterialExpressionNamedRerouteDeclaration_3 {
   class=MaterialExpressionNamedRerouteDeclaration
   inputs [None {from=MaterialExpressionMultiply_1}]
   properties {name=VSU, nodeColor=<rgba 0 0.9062 1 1>}
  }
  MaterialExpressionNamedRerouteDeclaration_4 {
   class=MaterialExpressionNamedRerouteDeclaration
   inputs [None {from=MaterialExpressionMultiply_3}]
   properties {name=VSW, nodeColor=<rgba 0 0.9062 1 1>}
  }
  MaterialExpressionNamedRerouteUsage_18 {class=MaterialExpressionNamedRerouteUsage}
  MaterialExpressionNamedRerouteUsage_19 {class=MaterialExpressionNamedRerouteUsage}
  MaterialExpressionNamedRerouteUsage_2 {class=MaterialExpressionNamedRerouteUsage}
  MaterialExpressionNamedRerouteUsage_20 {class=MaterialExpressionNamedRerouteUsage}
  MaterialExpressionNamedRerouteUsage_3 {class=MaterialExpressionNamedRerouteUsage}
  MaterialExpressionNamedRerouteUsage_4 {class=MaterialExpressionNamedRerouteUsage}
  MaterialExpressionNamedRerouteUsage_5 {class=MaterialExpressionNamedRerouteUsage}
  MaterialExpressionNamedRerouteUsage_6 {class=MaterialExpressionNamedRerouteUsage}
  MaterialExpressionNamedRerouteUsage_7 {class=MaterialExpressionNamedRerouteUsage}
  MaterialExpressionNamedRerouteUsage_8 {class=MaterialExpressionNamedRerouteUsage}
  MaterialExpressionSaturate_0 {class=MaterialExpressionSaturate, inputs [None {from=MaterialExpressionConvert_6}]}
  MaterialExpressionSine_0 {class=MaterialExpressionSine, inputs [None {from=MaterialExpressionFunctionInput_1}]}
 }}
