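# Parent Substrate OpenPBR opaque surface material, exposing base, emission, cutout, subsurface, coat, fuzz, iridescence and anisotropy controls.
# A static switch picks mesh UV or dithered triplanar coordinates, supplying shared UVs, explicit derivatives, tangent frame and rotation.
# These drive five texture samples (BaseColor, BasePack, Normal, EffectsPack, CoatPack), remapped by parameters into MF_Substrate_OpenPBR_Opaque.
# Base and coat normals are flippable and frame-oriented; coat has its own UV multiplier and Beer's-law tint.
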Material /BaseMaterial/Materials/Parents/M_OpenPBR_Opaque_Simple
graph {
 expressions {
  MaterialExpressionAdd_0 {
   class=MaterialExpressionAdd
   inputs [
    A {from=MaterialExpressionComponentMask_6}
    B {from=MaterialExpressionScalarParameter_13}
   ]
  }
  MaterialExpressionAdd_2 {
   class=MaterialExpressionAdd
   inputs [
    A {from=MaterialExpressionTextureSample_1, output=B}
    B {from=MaterialExpressionScalarParameter_9}
   ]
  }
  MaterialExpressionAppendVector_0 {
   class=MaterialExpressionAppendVector
   inputs [
    A {from=MaterialExpressionNamedRerouteUsage_29, output=""}
    B {from=MaterialExpressionNamedRerouteUsage_29, output=""}
   ]
  }
  MaterialExpressionChannelMaskParameter_1 {
   class=MaterialExpressionChannelMaskParameter
   inputs [None {from=MaterialExpressionTextureSample_0, output=RGBA}]
   properties {group=Base|Emission, parameterName=EmissionMask, sortPriority=0}
  }
  MaterialExpressionComponentMask_0 {class=MaterialExpressionComponentMask, inputs [None {from=MaterialExpressionTextureSample_3, output=RGB}], properties {g=true, r=true}}
  MaterialExpressionComponentMask_1 {class=MaterialExpressionComponentMask, inputs [None {from=MaterialExpressionVectorParameter_3}], properties {b=true, g=true, r=true}}
  MaterialExpressionComponentMask_2 {class=MaterialExpressionComponentMask, inputs [None {from=MaterialExpressionNamedRerouteUsage_11}], properties {g=true, r=true}}
  MaterialExpressionComponentMask_3 {class=MaterialExpressionComponentMask, inputs [None {from=MaterialExpressionReroute_4}], properties {a=true}}
  MaterialExpressionComponentMask_4 {class=MaterialExpressionComponentMask, inputs [None {from=MaterialExpressionReroute_6}], properties {a=true}}
  MaterialExpressionComponentMask_5 {class=MaterialExpressionComponentMask, inputs [None {from=MaterialExpressionVectorParameter_9}], properties {g=true, r=true}}
  MaterialExpressionComponentMask_6 {class=MaterialExpressionComponentMask, inputs [None {from=MaterialExpressionReroute_4}], properties {b=true}}
  MaterialExpressionComponentMask_7 {class=MaterialExpressionComponentMask, inputs [None {from=MaterialExpressionTextureSample_1, output=RGBA}], properties {a=true}}
  MaterialExpressionConstant2Vector_0 {class=MaterialExpressionConstant2Vector, properties {g=1, r=-1}}
  MaterialExpressionConstant2Vector_1 {class=MaterialExpressionConstant2Vector, properties {g=-1, r=1}}
  MaterialExpressionConstant2Vector_2 {class=MaterialExpressionConstant2Vector, properties {g=1, r=-1}}
  MaterialExpressionConstant2Vector_3 {class=MaterialExpressionConstant2Vector, properties {g=-1, r=1}}
  MaterialExpressionConstant_0 {class=MaterialExpressionConstant, properties {r=100}}
  MaterialExpressionConstant_11 {class=MaterialExpressionConstant, properties {r=1}}
  MaterialExpressionConstant_12 {class=MaterialExpressionConstant, properties {r=0.25}}
  MaterialExpressionConstant_7 {class=MaterialExpressionConstant, properties {r=200}}
  MaterialExpressionConvert_0 {
   class=MaterialExpressionConvert
   inputs [None {from=MaterialExpressionNamedRerouteUsage_13}]
   properties {
    convertInputs [{
      type=Vector4
      defaultValue=<rgba 0 0 0 1>
      expressionInput {
       expression=&self:MaterialExpressionNamedRerouteUsage_13
       inputName=None
       mask=0
       maskA=0
       maskB=0
       maskG=0
       maskR=0
       outputIndex=0
      }
     }]
    convertOutputs [{type=Vector2}, {type=Vector2}]
   }
  }
  MaterialExpressionConvert_1 {
   class=MaterialExpressionConvert
   inputs [None {from=MaterialExpressionNamedRerouteUsage_15}]
   properties {
    convertInputs [{
      type=Vector4
      defaultValue=<rgba 0 0 0 1>
      expressionInput {
       expression=&self:MaterialExpressionNamedRerouteUsage_15
       inputName=None
       mask=0
       maskA=0
       maskB=0
       maskG=0
       maskR=0
       outputIndex=0
      }
     }]
    convertOutputs [{type=Vector2}, {type=Vector2}]
   }
  }
  MaterialExpressionConvert_2 {
   class=MaterialExpressionConvert
   inputs [
    None {from=MaterialExpressionScalarParameter_3}
    None {from=MaterialExpressionScalarParameter_4}
    None {from=MaterialExpressionScalarParameter_5}
    None {from=MaterialExpressionScalarParameter_6}
   ]
   properties {
    convertInputs [
     {
      type=Scalar
      expressionInput {
       expression=&self:MaterialExpressionScalarParameter_3
       inputName=None
       mask=0
       maskA=0
       maskB=0
       maskG=0
       maskR=0
       outputIndex=0
      }
     }
     {
      type=Scalar
      expressionInput {
       expression=&self:MaterialExpressionScalarParameter_4
       inputName=None
       mask=0
       maskA=0
       maskB=0
       maskG=0
       maskR=0
       outputIndex=0
      }
     }
     {
      type=Scalar
      expressionInput {
       expression=&self:MaterialExpressionScalarParameter_5
       inputName=None
       mask=0
       maskA=0
       maskB=0
       maskG=0
       maskR=0
       outputIndex=0
      }
     }
     {
      type=Scalar
      expressionInput {
       expression=&self:MaterialExpressionScalarParameter_6
       inputName=None
       mask=0
       maskA=0
       maskB=0
       maskG=0
       maskR=0
       outputIndex=0
      }
     }
    ]
    convertOutputs [{type=Vector2}, {type=Vector2}]
   }
  }
  MaterialExpressionConvert_23 {
   class=MaterialExpressionConvert
   inputs [None {from=MaterialExpressionNamedRerouteUsage_1}]
   properties {
    convertInputs [{
      type=Vector4
      defaultValue=<rgba 0 0 0 1>
      expressionInput {
       expression=&self:MaterialExpressionNamedRerouteUsage_1
       inputName=None
       mask=0
       maskA=0
       maskB=0
       maskG=0
       maskR=0
       outputIndex=0
      }
     }]
    convertOutputs [{type=Vector2}, {type=Vector2}]
   }
  }
  MaterialExpressionConvert_4 {
   class=MaterialExpressionConvert
   inputs [None {from=MaterialExpressionNamedRerouteUsage_8}]
   properties {
    convertInputs [{
      type=Vector4
      defaultValue=<rgba 0 0 0 1>
      expressionInput {
       expression=&self:MaterialExpressionNamedRerouteUsage_8
       inputName=None
       mask=0
       maskA=0
       maskB=0
       maskG=0
       maskR=0
       outputIndex=0
      }
     }]
    convertOutputs [{type=Vector2}, {type=Vector2}]
   }
  }
  MaterialExpressionConvert_7 {
   class=MaterialExpressionConvert
   inputs [None {from=MaterialExpressionMultiply_14}]
   properties {
    convertInputs [{
      type=Vector4
      defaultValue=<rgba 0 0 0 1>
      expressionInput {
       expression=&self:MaterialExpressionMultiply_14
       inputName=None
       mask=0
       maskA=0
       maskB=0
       maskG=0
       maskR=0
       outputIndex=0
      }
     }]
    convertOutputs [{type=Vector2}, {type=Vector2}]
   }
  }
  MaterialExpressionDeriveNormalZ_0 {class=MaterialExpressionDeriveNormalZ, inputs [InXY {from=MaterialExpressionMultiply_0, output=""}], properties {}}
  MaterialExpressionDeriveNormalZ_1 {class=MaterialExpressionDeriveNormalZ, inputs [InXY {from=MaterialExpressionMultiply_6, output=""}], properties {}}
  MaterialExpressionDivide_0 {
   class=MaterialExpressionDivide
   inputs [
    A {from=MaterialExpressionMultiply_2}
    B {from=MaterialExpressionConstant_0}
   ]
  }
  MaterialExpressionDivide_1 {
   class=MaterialExpressionDivide
   inputs [
    A {from=MaterialExpressionVectorParameter_1, output=RGB}
    B {from=MaterialExpressionMaterialFunctionCall_23, output=Result}
   ]
  }
  MaterialExpressionIf_4 {
   class=MaterialExpressionIf
   inputs [
    A {from=MaterialExpressionSaturate_2, output=""}
    B {from="", output=""}
    "A > B" {from=MaterialExpressionScalarParameter_34, output=""}
    "A == B" {from="", output=""}
    "A < B" {from=MaterialExpressionConstant_11, output=""}
   ]
   properties {constB=0.0033}
  }
  MaterialExpressionLinearInterpolate_0 {
   class=MaterialExpressionLinearInterpolate
   inputs [
    A {from=MaterialExpressionReroute_2}
    B {from=MaterialExpressionSaturate_1}
    Alpha {from=MaterialExpressionVectorParameter_0, output=Tint Strength}
   ]
  }
  MaterialExpressionLinearInterpolate_1 {
   class=MaterialExpressionLinearInterpolate
   inputs [
    A {from=""}
    B {from=MaterialExpressionComponentMask_4}
    Alpha {from=MaterialExpressionScalarParameter_14}
   ]
   properties {constA=1}
  }
  MaterialExpressionLinearInterpolate_2 {
   class=MaterialExpressionLinearInterpolate
   inputs [
    A {from=MaterialExpressionScalarParameter_30}
    B {from=MaterialExpressionScalarParameter_31}
    Alpha {from=MaterialExpressionTextureSample_1, output=R}
   ]
  }
  MaterialExpressionLinearInterpolate_3 {
   class=MaterialExpressionLinearInterpolate
   inputs [
    A {from=MaterialExpressionTextureSample_0, output=RGB}
    B {from=MaterialExpressionMaterialFunctionCall_21, output=Out_Color}
    Alpha {from=MaterialExpressionScalarParameter_29}
   ]
  }
  MaterialExpressionLinearInterpolate_4 {
   class=MaterialExpressionLinearInterpolate
   inputs [
    A {from=MaterialExpressionScalarParameter_12}
    B {from=MaterialExpressionScalarParameter_19}
    Alpha {from=MaterialExpressionTextureSample_2, output=G}
   ]
  }
  MaterialExpressionLinearInterpolate_5 {
   class=MaterialExpressionLinearInterpolate
   inputs [
    A {from=MaterialExpressionScalarParameter_10}
    B {from=MaterialExpressionScalarParameter_11}
    Alpha {from=MaterialExpressionTextureSample_1, output=G}
   ]
  }
  MaterialExpressionLinearInterpolate_6 {
   class=MaterialExpressionLinearInterpolate
   inputs [
    A {from=MaterialExpressionScalarParameter_16}
    B {from=MaterialExpressionScalarParameter_20}
    Alpha {from=MaterialExpressionTextureSample_4, output=B}
   ]
  }
  MaterialExpressionLinearInterpolate_7 {
   class=MaterialExpressionLinearInterpolate
   inputs [
    A {from=MaterialExpressionScalarParameter_15}
    B {from=MaterialExpressionScalarParameter_26}
    Alpha {from=MaterialExpressionComponentMask_7}
   ]
  }
  MaterialExpressionLinearInterpolate_8 {
   class=MaterialExpressionLinearInterpolate
   inputs [
    A {from=MaterialExpressionVectorParameter_8, output=RGB}
    B {from=MaterialExpressionMultiply_18}
    Alpha {from=MaterialExpressionVectorParameter_8, output=RGB}
   ]
  }
  MaterialExpressionLinearInterpolate_9 {
   class=MaterialExpressionLinearInterpolate
   inputs [
    A {from=MaterialExpressionMultiply_19}
    B {from=MaterialExpressionVectorParameter_10, output=RGB}
    Alpha {from=MaterialExpressionVectorParameter_10, output=RGB}
   ]
  }
  MaterialExpressionMaterialFunctionCall_0 {
   class=MaterialExpressionMaterialFunctionCall
   inputs [
    ----| Base |---- {from=""}
    base_weight {from=MaterialExpressionMax_0}
    base_color {from=MaterialExpressionLinearInterpolate_0}
    base_diffuse_roughness {from=""}
    base_metalness {from=MaterialExpressionSaturate_4}
    ----| Specular |---- {from=""}
    specular_weight {from=""}
    specular_color {from=MaterialExpressionSaturate_6}
    specular_roughness {from=MaterialExpressionSaturate_0}
    specular_ior {from=MaterialExpressionConstant_7}
    specular_roughness_anisotropy {from=MaterialExpressionAdd_0}
    ----| Transmission |---- {from=""}
    transmission_weight {from=""}
    transmission_color {from=""}
    transmission_depth {from=""}
    transmission_scatter {from=""}
    transmission_scatter_anisotropy {from=""}
    transmission_dispersion_scale {from=""}
    transmission_dispersion_abbe_number {from=""}
    ----| Subsurface |---- {from=""}
    subsurface_weight {from=MaterialExpressionMultiply_5}
    subsurface_color {from=MaterialExpressionLinearInterpolate_8}
    subsurface_radius {from=MaterialExpressionLinearInterpolate_7}
    subsurface_radius_scale {from=""}
    subsurface_scatter_anisotropy {from=""}
    ----| Coat |---- {from=""}
    coat_weight {from=MaterialExpressionSaturate_2}
    coat_color {from=MaterialExpressionDivide_1}
    coat_roughness {from=MaterialExpressionSaturate_5}
    coat_roughness_anisotropy {from=""}
    coat_ior {from=MaterialExpressionReroute_15}
    coat_darkening {from=""}
    ----| Fuzz |---- {from=""}
    fuzz_weight {from=MaterialExpressionReroute_3}
    fuzz_color {from=MaterialExpressionLinearInterpolate_9}
    fuzz_roughness {from=MaterialExpressionReroute_1}
    ----| Emission |---- {from=""}
    emission_luminance {from=MaterialExpressionScalarParameter_35}
    emission_color {from=MaterialExpressionMultiply_20}
    ----| Thin Film |---- {from=""}
    thin_film_weight {from=""}
    thin_film_thickness {from=MaterialExpressionLinearInterpolate_4}
    thin_film_ior {from=MaterialExpressionScalarParameter_17}
    ----| Geometry |---- {from=""}
    geometry_opacity {from=MaterialExpressionLinearInterpolate_1}
    geometry_thin_walled {from=""}
    geometry_normal {from=MaterialExpressionNamedRerouteDeclaration_6}
    geometry_tangent {from=MaterialExpressionMaterialFunctionCall_11, output=Tangent}
    geometry_coat_normal {from=MaterialExpressionMaterialFunctionCall_18, output=TSN}
    geometry_coat_tangent {from=""}
   ]
   properties {materialFunction=/Engine/Functions/Substrate/MF_Substrate_OpenPBR_Opaque}
  }
  MaterialExpressionMaterialFunctionCall_1 {
   class=MaterialExpressionMaterialFunctionCall
   inputs [
    Selector {from=MaterialExpressionScalarParameter_1}
    VectorA {from=MaterialExpressionScalarParameter_0}
    VectorB {from=MaterialExpressionMultiply_1}
    VectorC {from=MaterialExpressionMultiply_3}
    VectorD {from=MaterialExpressionMultiply_4}
   ]
   properties {materialFunction=/BaseMaterial/Materials/Functions/MF_Switch4_Vec2}
  }
  MaterialExpressionMaterialFunctionCall_10 {
   class=MaterialExpressionMaterialFunctionCall
   inputs [
    TSN {from=MaterialExpressionDeriveNormalZ_0}
    Tangent {from=MaterialExpressionNamedRerouteUsage_2}
    Bitangent {from=MaterialExpressionNamedRerouteUsage_3}
    "Rotation (Turns)" {from=MaterialExpressionNamedRerouteUsage_4}
   ]
   properties {materialFunction=/BaseMaterial/Materials/Functions/MF_TangentNormal_OrientWorld}
  }
  MaterialExpressionMaterialFunctionCall_11 {
   class=MaterialExpressionMaterialFunctionCall
   inputs [Rotation {from=MaterialExpressionComponentMask_3}]
   properties {materialFunction=/Engine/Functions/Substrate/Substrate-Rotation-To-Tangent}
  }
  MaterialExpressionMaterialFunctionCall_18 {
   class=MaterialExpressionMaterialFunctionCall
   inputs [
    TSN {from=MaterialExpressionDeriveNormalZ_1}
    Tangent {from=MaterialExpressionNamedRerouteUsage_25}
    Bitangent {from=MaterialExpressionNamedRerouteUsage_26}
    "Rotation (Turns)" {from=MaterialExpressionNamedRerouteUsage_27}
   ]
   properties {materialFunction=&MF_TangentNormal_OrientWorld}
  }
  MaterialExpressionMaterialFunctionCall_19 {
   class=MaterialExpressionMaterialFunctionCall
   inputs [
    Selector {from=MaterialExpressionScalarParameter_24}
    VectorA {from=MaterialExpressionScalarParameter_23}
    VectorB {from=MaterialExpressionMultiply_7}
    VectorC {from=MaterialExpressionMultiply_11}
    VectorD {from=MaterialExpressionMultiply_12}
   ]
   properties {materialFunction=&MF_Switch4_Vec2}
  }
  MaterialExpressionMaterialFunctionCall_2 {
   class=MaterialExpressionMaterialFunctionCall
   inputs [
    CoordinateRepeat {from=MaterialExpressionConvert_2}
    CoordinateTransform {from=MaterialExpressionConvert_2}
    "Rotation (Turns)" {from=MaterialExpressionScalarParameter_18}
    UVIndex {from=MaterialExpressionScalarParameter_2}
   ]
   properties {materialFunction=/BaseMaterial/Materials/Functions/MF_ResolveCoordinateFrame_SimpleUV_Basic}
  }
  MaterialExpressionMaterialFunctionCall_21 {
   class=MaterialExpressionMaterialFunctionCall
   inputs [
    In_Color {from=MaterialExpressionTextureSample_0, output=RGB}
    In_Hue {from=MaterialExpressionScalarParameter_25}
    In_Saturation {from=MaterialExpressionScalarParameter_27}
    In_Value {from=MaterialExpressionScalarParameter_28}
   ]
   properties {materialFunction=/BaseMaterial/Materials/Functions/MF_TransformHSV}
  }
  MaterialExpressionMaterialFunctionCall_23 {
   class=MaterialExpressionMaterialFunctionCall
   inputs [
    Thickness {from=MaterialExpressionScalarParameter_32}
    Depth Scale {from=MaterialExpressionConstant_12}
   ]
   properties {materialFunction=/Engine/Functions/Engine_MaterialFunctions01/Volumetrics/BeersLaw}
  }
  MaterialExpressionMaterialFunctionCall_3 {
   class=MaterialExpressionMaterialFunctionCall
   inputs [
    CoordinateSpace {from=MaterialExpressionScalarParameter_7}
    CoordinateRepeat {from=MaterialExpressionDivide_0}
    CoordinateTransform {from=MaterialExpressionComponentMask_1}
    "Rotation (Turns)" {from=MaterialExpressionScalarParameter_18}
    Dither {from=""}
    BlendFactor {from=MaterialExpressionScalarParameter_8}
   ]
   properties {materialFunction=/BaseMaterial/Materials/Functions/MF_ResolveCoordinateFrame_Triplanar_Dither}
  }
  MaterialExpressionMax_0 {
   class=MaterialExpressionMax
   inputs [
    A {from=MaterialExpressionOneMinus_0}
    B {from=MaterialExpressionReroute_11}
   ]
  }
  MaterialExpressionMax_1 {
   class=MaterialExpressionMax
   inputs [
    A {from=MaterialExpressionVectorParameter_11, output=RGB}
    B {from=""}
   ]
   properties {constB=0}
  }
  MaterialExpressionMultiply_0 {
   class=MaterialExpressionMultiply
   inputs [
    A {from=MaterialExpressionComponentMask_0}
    B {from=MaterialExpressionMaterialFunctionCall_1, output=Selected Vector}
   ]
  }
  MaterialExpressionMultiply_1 {
   class=MaterialExpressionMultiply
   inputs [
    A {from=MaterialExpressionScalarParameter_0}
    B {from=MaterialExpressionConstant2Vector_0}
   ]
  }
  MaterialExpressionMultiply_10 {
   class=MaterialExpressionMultiply
   inputs [
    A {from=MaterialExpressionVectorParameter_4, output=RGB}
    B {from=MaterialExpressionLinearInterpolate_2}
   ]
  }
  MaterialExpressionMultiply_11 {
   class=MaterialExpressionMultiply
   inputs [
    A {from=MaterialExpressionScalarParameter_23}
    B {from=MaterialExpressionConstant2Vector_3}
   ]
  }
  MaterialExpressionMultiply_12 {
   class=MaterialExpressionMultiply
   inputs [
    A {from=MaterialExpressionScalarParameter_23}
    B {from=""}
   ]
   properties {constB=-1}
  }
  MaterialExpressionMultiply_13 {
   class=MaterialExpressionMultiply
   inputs [
    A {from=MaterialExpressionNamedRerouteUsage_19}
    B {from=MaterialExpressionNamedRerouteUsage_28}
   ]
  }
  MaterialExpressionMultiply_14 {
   class=MaterialExpressionMultiply
   inputs [
    A {from=MaterialExpressionNamedRerouteUsage_20}
    B {from=MaterialExpressionAppendVector_0}
   ]
  }
  MaterialExpressionMultiply_15 {
   class=MaterialExpressionMultiply
   inputs [
    A {from=MaterialExpressionTextureSample_4, output=A}
    B {from=MaterialExpressionScalarParameter_33}
   ]
  }
  MaterialExpressionMultiply_17 {
   class=MaterialExpressionMultiply
   inputs [
    A {from=MaterialExpressionComponentMask_5}
    B {from=MaterialExpressionVectorParameter_9, output=Uniform Scale}
   ]
  }
  MaterialExpressionMultiply_18 {
   class=MaterialExpressionMultiply
   inputs [
    A {from=MaterialExpressionLinearInterpolate_0}
    B {from=MaterialExpressionVectorParameter_8, output=RGB}
   ]
  }
  MaterialExpressionMultiply_19 {
   class=MaterialExpressionMultiply
   inputs [
    A {from=MaterialExpressionLinearInterpolate_0}
    B {from=MaterialExpressionVectorParameter_10, output=RGB}
   ]
  }
  MaterialExpressionMultiply_2 {
   class=MaterialExpressionMultiply
   inputs [
    A {from=MaterialExpressionVectorParameter_2}
    B {from=MaterialExpressionVectorParameter_2}
   ]
  }
  MaterialExpressionMultiply_20 {
   class=MaterialExpressionMultiply
   inputs [
    A {from=MaterialExpressionMax_1}
    B {from=MaterialExpressionChannelMaskParameter_1, output=R}
   ]
  }
  MaterialExpressionMultiply_3 {
   class=MaterialExpressionMultiply
   inputs [
    A {from=MaterialExpressionScalarParameter_0}
    B {from=MaterialExpressionConstant2Vector_1}
   ]
  }
  MaterialExpressionMultiply_4 {
   class=MaterialExpressionMultiply
   inputs [
    A {from=MaterialExpressionScalarParameter_0}
    B {from=""}
   ]
   properties {constB=-1}
  }
  MaterialExpressionMultiply_5 {
   class=MaterialExpressionMultiply
   inputs [
    A {from=MaterialExpressionScalarParameter_22}
    B {from=MaterialExpressionOneMinus_1}
   ]
  }
  MaterialExpressionMultiply_6 {
   class=MaterialExpressionMultiply
   inputs [
    A {from=MaterialExpressionComponentMask_2}
    B {from=MaterialExpressionMaterialFunctionCall_19, output=Selected Vector}
   ]
  }
  MaterialExpressionMultiply_7 {
   class=MaterialExpressionMultiply
   inputs [
    A {from=MaterialExpressionScalarParameter_23}
    B {from=MaterialExpressionConstant2Vector_2}
   ]
  }
  MaterialExpressionMultiply_8 {
   class=MaterialExpressionMultiply
   inputs [
    A {from=MaterialExpressionLinearInterpolate_3}
    B {from=MaterialExpressionVectorParameter_0, output=RGB}
   ]
  }
  MaterialExpressionNamedRerouteDeclaration_0 {
   class=MaterialExpressionNamedRerouteDeclaration
   inputs [None {from=MaterialExpressionStaticSwitch_0}]
   properties {name=Turns, nodeColor=<rgba 0.99 0.0099 0.1079 1>}
  }
  MaterialExpressionNamedRerouteDeclaration_1 {
   class=MaterialExpressionNamedRerouteDeclaration
   inputs [None {from=MaterialExpressionStaticSwitch_1}]
   properties {name=UV, nodeColor=<rgba 0.99 0.0099 0.1079 1>}
  }
  MaterialExpressionNamedRerouteDeclaration_2 {
   class=MaterialExpressionNamedRerouteDeclaration
   inputs [None {from=MaterialExpressionStaticSwitch_2}]
   properties {name=T, nodeColor=<rgba 0.99 0.0099 0.1079 1>}
  }
  MaterialExpressionNamedRerouteDeclaration_3 {
   class=MaterialExpressionNamedRerouteDeclaration
   inputs [None {from=MaterialExpressionStaticSwitch_3}]
   properties {name=B, nodeColor=<rgba 0.99 0.0099 0.1079 1>}
  }
  MaterialExpressionNamedRerouteDeclaration_4 {
   class=MaterialExpressionNamedRerouteDeclaration
   inputs [None {from=MaterialExpressionStaticSwitch_4}]
   properties {name=DXY, nodeColor=<rgba 0.99 0.0099 0.1079 1>}
  }
  MaterialExpressionNamedRerouteDeclaration_5 {
   class=MaterialExpressionNamedRerouteDeclaration
   inputs [None {from=MaterialExpressionTextureSample_4, output=RGB}]
   properties {name=TopSample, nodeColor=<rgba 1 0 0.4922 1>}
  }
  MaterialExpressionNamedRerouteDeclaration_6 {
   class=MaterialExpressionNamedRerouteDeclaration
   inputs [None {from=MaterialExpressionMaterialFunctionCall_10, output=TSN}]
   properties {name=Normal, nodeColor=<rgba 0 0.3607 1 1>}
  }
  MaterialExpressionNamedRerouteDeclaration_7 {
   class=MaterialExpressionNamedRerouteDeclaration
   inputs [None {from=MaterialExpressionMultiply_17}]
   properties {name=CoatUVMult, nodeColor=<rgba 0.99 0.0099 0.1079 1>}
  }
  MaterialExpressionNamedRerouteUsage_0 {class=MaterialExpressionNamedRerouteUsage}
  MaterialExpressionNamedRerouteUsage_1 {class=MaterialExpressionNamedRerouteUsage}
  MaterialExpressionNamedRerouteUsage_11 {class=MaterialExpressionNamedRerouteUsage}
  MaterialExpressionNamedRerouteUsage_12 {class=MaterialExpressionNamedRerouteUsage}
  MaterialExpressionNamedRerouteUsage_13 {class=MaterialExpressionNamedRerouteUsage}
  MaterialExpressionNamedRerouteUsage_15 {class=MaterialExpressionNamedRerouteUsage}
  MaterialExpressionNamedRerouteUsage_16 {class=MaterialExpressionNamedRerouteUsage}
  MaterialExpressionNamedRerouteUsage_19 {class=MaterialExpressionNamedRerouteUsage}
  MaterialExpressionNamedRerouteUsage_2 {class=MaterialExpressionNamedRerouteUsage}
  MaterialExpressionNamedRerouteUsage_20 {class=MaterialExpressionNamedRerouteUsage}
  MaterialExpressionNamedRerouteUsage_25 {class=MaterialExpressionNamedRerouteUsage}
  MaterialExpressionNamedRerouteUsage_26 {class=MaterialExpressionNamedRerouteUsage}
  MaterialExpressionNamedRerouteUsage_27 {class=MaterialExpressionNamedRerouteUsage}
  MaterialExpressionNamedRerouteUsage_28 {class=MaterialExpressionNamedRerouteUsage}
  MaterialExpressionNamedRerouteUsage_29 {class=MaterialExpressionNamedRerouteUsage}
  MaterialExpressionNamedRerouteUsage_3 {class=MaterialExpressionNamedRerouteUsage}
  MaterialExpressionNamedRerouteUsage_4 {class=MaterialExpressionNamedRerouteUsage}
  MaterialExpressionNamedRerouteUsage_7 {class=MaterialExpressionNamedRerouteUsage}
  MaterialExpressionNamedRerouteUsage_8 {class=MaterialExpressionNamedRerouteUsage}
  MaterialExpressionOneMinus_0 {class=MaterialExpressionOneMinus, inputs [None {from=MaterialExpressionScalarParameter_22}]}
  MaterialExpressionOneMinus_1 {class=MaterialExpressionOneMinus, inputs [None {from=MaterialExpressionReroute_11}]}
  MaterialExpressionReroute_1 {class=MaterialExpressionReroute, inputs [None {from=MaterialExpressionScalarParameter_21}]}
  MaterialExpressionReroute_11 {class=MaterialExpressionReroute, inputs [None {from=MaterialExpressionSaturate_4}]}
  MaterialExpressionReroute_15 {class=MaterialExpressionReroute, inputs [None {from=MaterialExpressionIf_4}]}
  MaterialExpressionReroute_2 {class=MaterialExpressionReroute, inputs [None {from=MaterialExpressionLinearInterpolate_3}]}
  MaterialExpressionReroute_3 {class=MaterialExpressionReroute, inputs [None {from=MaterialExpressionReroute_7}]}
  MaterialExpressionReroute_4 {class=MaterialExpressionReroute, inputs [None {from=MaterialExpressionTextureSample_2, output=RGBA}]}
  MaterialExpressionReroute_6 {class=MaterialExpressionReroute, inputs [None {from=MaterialExpressionTextureSample_0, output=RGBA}]}
  MaterialExpressionReroute_7 {class=MaterialExpressionReroute, inputs [None {from=MaterialExpressionTextureSample_2, output=R}]}
  MaterialExpressionSaturate_0 {class=MaterialExpressionSaturate, inputs [None {from=MaterialExpressionLinearInterpolate_5}]}
  MaterialExpressionSaturate_1 {class=MaterialExpressionSaturate, inputs [None {from=MaterialExpressionMultiply_8}]}
  MaterialExpressionSaturate_2 {class=MaterialExpressionSaturate, inputs [None {from=MaterialExpressionMultiply_15}]}
  MaterialExpressionSaturate_4 {class=MaterialExpressionSaturate, inputs [None {from=MaterialExpressionAdd_2}]}
  MaterialExpressionSaturate_5 {class=MaterialExpressionSaturate, inputs [None {from=MaterialExpressionLinearInterpolate_6}]}
  MaterialExpressionSaturate_6 {class=MaterialExpressionSaturate, inputs [None {from=MaterialExpressionMultiply_10}]}
  MaterialExpressionScalarParameter_0 {
   class=MaterialExpressionScalarParameter
   properties {defaultValue=1, group=Base|Geometry, parameterName=BaseNormal, sliderMax=2, sliderMin=-2, sortPriority=10}
  }
  MaterialExpressionScalarParameter_1 {
   class=MaterialExpressionScalarParameter
   properties {
    controlType=Enumeration
    enumeration=/BaseMaterial/Materials/Utilities/ENormalFlip
    group=Base|Geometry
    parameterName=BaseNormalFlip
    sortPriority=20
   }
  }
  MaterialExpressionScalarParameter_10 {class=MaterialExpressionScalarParameter, properties {group=Base|Specular, parameterName=RoughnessMin, sortPriority=20}}
  MaterialExpressionScalarParameter_11 {class=MaterialExpressionScalarParameter, properties {defaultValue=1, group=Base|Specular, parameterName=RoughnessMax, sortPriority=21}}
  MaterialExpressionScalarParameter_12 {class=MaterialExpressionScalarParameter, properties {group=Effects|Iridescence, parameterName=FilmThicknessMin, sortPriority=10}}
  MaterialExpressionScalarParameter_13 {
   class=MaterialExpressionScalarParameter
   properties {group=Effects|Anisotropy, parameterName=Anisotropy, sliderMax=1, sliderMin=-1, sortPriority=30}
  }
  MaterialExpressionScalarParameter_14 {class=MaterialExpressionScalarParameter, properties {group=Base|Geometry, parameterName=CutoutChannelWeight, sortPriority=30}}
  MaterialExpressionScalarParameter_15 {
   class=MaterialExpressionScalarParameter
   properties {defaultValue=1, group=Effects|Subsurface, parameterName=ScatterMin, sliderMax=10, sortPriority=20}
  }
  MaterialExpressionScalarParameter_16 {
   class=MaterialExpressionScalarParameter
   properties {defaultValue=0.15, group=Effects|Coat, parameterName=CoathRoughMin, sliderMax=1, sortPriority=20}
  }
  MaterialExpressionScalarParameter_17 {class=MaterialExpressionScalarParameter, properties {defaultValue=1.55, group=Effects|Iridescence, parameterName=FilmIOR, sortPriority=20}}
  MaterialExpressionScalarParameter_18 {
   class=MaterialExpressionScalarParameter
   properties {group=Sampling|Coordinates, parameterName=Rotation, sliderMax=1, sliderMin=-1, sortPriority=25}
  }
  MaterialExpressionScalarParameter_19 {class=MaterialExpressionScalarParameter, properties {group=Effects|Iridescence, parameterName=FilmThicknessMax, sortPriority=11}}
  MaterialExpressionScalarParameter_2 {
   class=MaterialExpressionScalarParameter
   properties {
    controlType=Enumeration
    enumeration=/BaseMaterial/Materials/Utilities/EUVCoordinateIndex
    group=Sampling|Coordinates
    parameterName=UVIndex
    sortPriority=0
   }
  }
  MaterialExpressionScalarParameter_20 {
   class=MaterialExpressionScalarParameter
   properties {defaultValue=0.15, group=Effects|Coat, parameterName=CoatRoughMax, sliderMax=1, sortPriority=21}
  }
  MaterialExpressionScalarParameter_21 {class=MaterialExpressionScalarParameter, properties {group=Effects|Fuzz, parameterName=Fuzz Roughness, sliderMax=1, sortPriority=20}}
  MaterialExpressionScalarParameter_22 {
   class=MaterialExpressionScalarParameter
   properties {group=Effects|Subsurface, parameterName=SubsurfaceIntensity, sliderMax=1, sortPriority=0}
  }
  MaterialExpressionScalarParameter_23 {
   class=MaterialExpressionScalarParameter
   properties {defaultValue=1, group=Effects|Coat, parameterName=CoatNormal, sliderMax=2, sliderMin=-2, sortPriority=30}
  }
  MaterialExpressionScalarParameter_24 {
   class=MaterialExpressionScalarParameter
   properties {
    controlType=Enumeration
    enumeration=&ENormalFlip
    group=Effects|Coat
    parameterName=CoatNormalFlip
    sortPriority=40
   }
  }
  MaterialExpressionScalarParameter_25 {class=MaterialExpressionScalarParameter, properties {group=Base|BaseColor, parameterName=HueShift, sliderMax=1, sortPriority=0}}
  MaterialExpressionScalarParameter_26 {
   class=MaterialExpressionScalarParameter
   properties {defaultValue=1, group=Effects|Subsurface, parameterName=ScatterMax, sliderMax=10, sortPriority=21}
  }
  MaterialExpressionScalarParameter_27 {
   class=MaterialExpressionScalarParameter
   properties {defaultValue=1, group=Base|BaseColor, parameterName=Saturation, sliderMax=2, sortPriority=2}
  }
  MaterialExpressionScalarParameter_28 {class=MaterialExpressionScalarParameter, properties {defaultValue=1, group=Base|BaseColor, parameterName=Value, sliderMax=2, sortPriority=4}}
  MaterialExpressionScalarParameter_29 {class=MaterialExpressionScalarParameter, properties {group=Base|BaseColor, parameterName=HSVStrength, sliderMax=1, sortPriority=-1}}
  MaterialExpressionScalarParameter_3 {
   class=MaterialExpressionScalarParameter
   properties {defaultValue=1, group=Sampling|Coordinates, parameterName=TileU, sliderMax=10, sortPriority=10}
  }
  MaterialExpressionScalarParameter_30 {class=MaterialExpressionScalarParameter, properties {group=Base|Specular, parameterName=SpecularMinF0, sliderMax=1, sortPriority=10}}
  MaterialExpressionScalarParameter_31 {
   class=MaterialExpressionScalarParameter
   properties {defaultValue=0.08, group=Base|Specular, parameterName=SpecularMaxF0, sliderMax=1, sortPriority=11}
  }
  MaterialExpressionScalarParameter_32 {class=MaterialExpressionScalarParameter, properties {defaultValue=0.5, group=Effects|Coat, parameterName=CoatThickness, sortPriority=10}}
  MaterialExpressionScalarParameter_33 {class=MaterialExpressionScalarParameter, properties {group=Effects|Coat, parameterName=CoatCoverage, sliderMax=1, sortPriority=-10}}
  MaterialExpressionScalarParameter_34 {
   class=MaterialExpressionScalarParameter
   properties {defaultValue=1.55, group=Effects|Coat, parameterName=CoatIOR, sliderMax=3, sliderMin=1, sortPriority=50}
  }
  MaterialExpressionScalarParameter_35 {class=MaterialExpressionScalarParameter, properties {group=Base|Emission, parameterName=EmissionLuminance, sliderMax=1000, sortPriority=20}}
  MaterialExpressionScalarParameter_4 {
   class=MaterialExpressionScalarParameter
   properties {defaultValue=1, group=Sampling|Coordinates, parameterName=TileV, sliderMax=10, sortPriority=11}
  }
  MaterialExpressionScalarParameter_5 {
   class=MaterialExpressionScalarParameter
   properties {group=Sampling|Coordinates, parameterName=OffsetU, sliderMax=1, sliderMin=-1, sortPriority=20}
  }
  MaterialExpressionScalarParameter_6 {
   class=MaterialExpressionScalarParameter
   properties {group=Sampling|Coordinates, parameterName=OffsetV, sliderMax=1, sliderMin=-1, sortPriority=21}
  }
  MaterialExpressionScalarParameter_7 {
   class=MaterialExpressionScalarParameter
   properties {
    controlType=Enumeration
    defaultValue=1
    enumeration=/BaseMaterial/Materials/Utilities/ETriplanarCoordinateFrame
    group=Sampling|Coordinates
    parameterName=CoordinateSpace
    sortPriority=0
   }
  }
  MaterialExpressionScalarParameter_8 {
   class=MaterialExpressionScalarParameter
   properties {defaultValue=0.5, group=Sampling|Coordinates, parameterName=TransitionSoftness, sliderMax=1, sortPriority=30}
  }
  MaterialExpressionScalarParameter_9 {
   class=MaterialExpressionScalarParameter
   properties {group=Base|Metalness, parameterName=Metalness, sliderMax=1, sliderMin=-1, sortPriority=0}
  }
  MaterialExpressionStaticBoolParameter_0 {
   class=MaterialExpressionStaticBoolParameter
   inputs []
   properties {falseLabel=MeshUV, group=Sampling|Coordinates, parameterName=CoordinateType, sortPriority=-10, trueLabel=Triplanar}
  }
  MaterialExpressionStaticSwitch_0 {
   class=MaterialExpressionStaticSwitch
   inputs [
    True {from=MaterialExpressionMaterialFunctionCall_3, output="Rotation (Turns)"}
    False {from=MaterialExpressionMaterialFunctionCall_2, output="Rotation (Turns)"}
    Value {from=MaterialExpressionStaticBoolParameter_0, output=""}
   ]
   properties {}
  }
  MaterialExpressionStaticSwitch_1 {
   class=MaterialExpressionStaticSwitch
   inputs [
    True {from=MaterialExpressionMaterialFunctionCall_3, output=UV}
    False {from=MaterialExpressionMaterialFunctionCall_2, output=UV}
    Value {from=MaterialExpressionStaticBoolParameter_0, output=""}
   ]
   properties {}
  }
  MaterialExpressionStaticSwitch_2 {
   class=MaterialExpressionStaticSwitch
   inputs [
    True {from=MaterialExpressionMaterialFunctionCall_3, output=Tangent}
    False {from=MaterialExpressionMaterialFunctionCall_2, output=Tangent}
    Value {from=MaterialExpressionStaticBoolParameter_0, output=""}
   ]
   properties {}
  }
  MaterialExpressionStaticSwitch_3 {
   class=MaterialExpressionStaticSwitch
   inputs [
    True {from=MaterialExpressionMaterialFunctionCall_3, output=Bitangent}
    False {from=MaterialExpressionMaterialFunctionCall_2, output=Bitangent}
    Value {from=MaterialExpressionStaticBoolParameter_0, output=""}
   ]
   properties {}
  }
  MaterialExpressionStaticSwitch_4 {
   class=MaterialExpressionStaticSwitch
   inputs [
    True {from=MaterialExpressionMaterialFunctionCall_3, output=DXY}
    False {from=MaterialExpressionMaterialFunctionCall_2, output=DXY}
    Value {from=MaterialExpressionStaticBoolParameter_0, output=""}
   ]
   properties {}
  }
  MaterialExpressionTextureObjectParameter_0 {
   class=MaterialExpressionTextureObjectParameter
   inputs [
    UVs {from="", output=""}
    Apply View MipBias {from="", output=""}
   ]
   properties {
    group=Sampling|Textures
    parameterName=Normal
    samplerType=SAMPLERTYPE_Normal
    sortPriority=20
    texture=/BaseMaterial/Textures/Default/T_FlatNorm_BC5
   }
  }
  MaterialExpressionTextureObjectParameter_2 {
   class=MaterialExpressionTextureObjectParameter
   inputs [
    UVs {from="", output=""}
    Apply View MipBias {from="", output=""}
   ]
   properties {
    group=Sampling|Textures
    parameterName=BaseColor
    sortPriority=0
    texture=/BaseMaterial/Textures/Default/T_Gray_sRGB
    channelNames {a=Opacity Mask, b="", g="", r=""}
   }
  }
  MaterialExpressionTextureObjectParameter_4 {
   class=MaterialExpressionTextureObjectParameter
   inputs [
    UVs {from="", output=""}
    Apply View MipBias {from="", output=""}
   ]
   properties {
    group=Sampling|Textures
    parameterName=BasePack
    samplerType=SAMPLERTYPE_Masks
    sortPriority=10
    texture=/BaseMaterial/Textures/Default/T_SRM_Mask
    channelNames {a=SSS Thickness, b=Metallic, g=Roughness, r=Specular}
   }
  }
  MaterialExpressionTextureObjectParameter_5 {
   class=MaterialExpressionTextureObjectParameter
   inputs [
    UVs {from="", output=""}
    Apply View MipBias {from="", output=""}
   ]
   properties {
    group=Sampling|Textures
    parameterName=EffectsPack
    samplerType=SAMPLERTYPE_LinearColor
    sortPriority=30
    texture=/BaseMaterial/Textures/Default/T_BlackLinearColor
    channelNames {a=Aniso Rotation, b=Anisotropy, g=Film Thickness, r=Fuzz Amount}
   }
  }
  MaterialExpressionTextureObjectParameter_7 {
   class=MaterialExpressionTextureObjectParameter
   inputs [
    UVs {from="", output=""}
    Apply View MipBias {from="", output=""}
   ]
   properties {
    group=Sampling|Textures
    parameterName=CoatPack
    samplerType=SAMPLERTYPE_LinearColor
    sortPriority=40
    texture=&T_BlackLinearColor
    channelNames {a=Coat Coverage, b=Coat Roughness, g=NormalY, r=NormalX}
   }
  }
  MaterialExpressionTextureSample_0 {
   class=MaterialExpressionTextureSample
   inputs [
    UVs {from=MaterialExpressionNamedRerouteUsage_7}
    Tex {from=MaterialExpressionTextureObjectParameter_2}
    "DDX(UVs)" {from=MaterialExpressionConvert_4}
    "DDY(UVs)" {from=MaterialExpressionConvert_4}
    Apply View MipBias {from=""}
   ]
   properties {mipValueMode=TMVM_Derivative, samplerSource=SSM_Wrap_WorldGroupSettings}
  }
  MaterialExpressionTextureSample_1 {
   class=MaterialExpressionTextureSample
   inputs [
    UVs {from=MaterialExpressionNamedRerouteUsage_12}
    Tex {from=MaterialExpressionTextureObjectParameter_4}
    "DDX(UVs)" {from=MaterialExpressionConvert_0}
    "DDY(UVs)" {from=MaterialExpressionConvert_0}
    Apply View MipBias {from=""}
   ]
   properties {mipValueMode=TMVM_Derivative, samplerSource=SSM_Wrap_WorldGroupSettings}
  }
  MaterialExpressionTextureSample_2 {
   class=MaterialExpressionTextureSample
   inputs [
    UVs {from=MaterialExpressionNamedRerouteUsage_16}
    Tex {from=MaterialExpressionTextureObjectParameter_5}
    "DDX(UVs)" {from=MaterialExpressionConvert_1}
    "DDY(UVs)" {from=MaterialExpressionConvert_1}
    Apply View MipBias {from=""}
   ]
   properties {mipValueMode=TMVM_Derivative, samplerSource=SSM_Wrap_WorldGroupSettings}
  }
  MaterialExpressionTextureSample_3 {
   class=MaterialExpressionTextureSample
   inputs [
    UVs {from=MaterialExpressionNamedRerouteUsage_0}
    Tex {from=MaterialExpressionTextureObjectParameter_0}
    "DDX(UVs)" {from=MaterialExpressionConvert_23}
    "DDY(UVs)" {from=MaterialExpressionConvert_23}
    Apply View MipBias {from=""}
   ]
   properties {mipValueMode=TMVM_Derivative, samplerSource=SSM_Wrap_WorldGroupSettings}
  }
  MaterialExpressionTextureSample_4 {
   class=MaterialExpressionTextureSample
   inputs [
    UVs {from=MaterialExpressionMultiply_13}
    Tex {from=MaterialExpressionTextureObjectParameter_7}
    "DDX(UVs)" {from=MaterialExpressionConvert_7}
    "DDY(UVs)" {from=MaterialExpressionConvert_7}
    Apply View MipBias {from=""}
   ]
   properties {mipValueMode=TMVM_Derivative, samplerSource=SSM_Wrap_WorldGroupSettings}
  }
  MaterialExpressionVectorParameter_0 {
   class=MaterialExpressionVectorParameter
   properties {group=Base|BaseColor, parameterName=Tint, sortPriority=10, channelNames {a=Tint Strength, b="", g="", r=""}, defaultValue=<rgba 1 1 1 0>}
  }
  MaterialExpressionVectorParameter_1 {
   class=MaterialExpressionVectorParameter
   properties {
    group=Effects|Coat
    parameterName=CoatColor
    sortPriority=0
    channelNames {a=None, b="", g="", r=""}
    defaultValue=<rgba 0.975 0.975 0.975 1>
   }
  }
  MaterialExpressionVectorParameter_10 {
   class=MaterialExpressionVectorParameter
   properties {group=Effects|Fuzz, parameterName=FuzzTint, sortPriority=10, channelNames {a=Only Use Swatch, b="", g="", r=""}, defaultValue=<rgba 1 1 1 1>}
  }
  MaterialExpressionVectorParameter_11 {
   class=MaterialExpressionVectorParameter
   properties {group=Base|Emission, parameterName=EmissionColor, sortPriority=10, channelNames {a=None, b="", g="", r=""}, defaultValue=<rgba 1 1 1 0>}
  }
  MaterialExpressionVectorParameter_2 {
   class=MaterialExpressionVectorParameter
   properties {
    group=Sampling|Coordinates
    parameterName=TileXYZ
    sortPriority=10
    channelNames {a=Uniform Scale, b=Tile Z, g=Tile Y, r=Tile X}
    defaultValue=<rgba 1 1 1 1>
   }
  }
  MaterialExpressionVectorParameter_3 {
   class=MaterialExpressionVectorParameter
   properties {
    group=Sampling|Coordinates
    parameterName=OffsetXYZ
    sortPriority=20
    channelNames {a=None, b=Translate Z, g=Translate Y, r=Translate X}
    defaultValue=<rgba 0 0 0 1>
   }
  }
  MaterialExpressionVectorParameter_4 {
   class=MaterialExpressionVectorParameter
   properties {group=Base|Specular, parameterName=ReflectionColor, sortPriority=0, defaultValue=<rgba 1 1 1 1>}
  }
  MaterialExpressionVectorParameter_8 {
   class=MaterialExpressionVectorParameter
   properties {
    group=Effects|Subsurface
    parameterName=SubsurfaceTint
    sortPriority=10
    channelNames {a=Only Use Swatch, b="", g="", r=""}
    defaultValue=<rgba 1 1 1 0>
   }
  }
  MaterialExpressionVectorParameter_9 {
   class=MaterialExpressionVectorParameter
   properties {
    group=Sampling|Coordinates
    parameterName=TopcoatTile
    sortPriority=40
    channelNames {a=None, b=Uniform Scale, g=Tile Y, r=Tile X}
    defaultValue=<rgba 1 1 1 1>
   }
  }
 }
 properties {
  bUsedWithInstancedStaticMeshes=true
  bUsedWithNanite=true
  bUsedWithSkeletalMesh=true
  shadingModel=MSM_SubsurfaceProfile
  usedShadingModels=Subsurface Profile
 }
}
